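# Player animation blueprint for the Deployable Turret item, supplying Item Override animation layers on the Fortnite player skeleton.
# Only ItemUpperBody has logic; the other layers and main AnimGraph are unwired stubs.
# It caches the incoming pose and, while riding or sliding, applies the matching riding-idle or slide-loop additive through inertialized bool blends.
# A non-looping slide additive player is present but unconnected, and the event graph has no logic.
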AnimBlueprint /DeployableTurretGameplay/Items/DeployableTurret/Animations/DeployableTurret_Player_AnimBP
target_skeleton=&Fortnite_M_Avg_Player_Skeleton
graphs [
 AnimGraph {nodes [Output Pose() {layerGroup=None}]}
 EventGraph {nodes [
   TryGetPawnOwner()
   Event BlueprintUpdateAnimation()
  ]}
 ItemAimOffset {nodes [
   Input Pose!() -> (Pose, AimOffsetAlpha:double, Yaw:double, Pitch:double) {name=InPose_AimOffset}
   Input Pose!() -> (Pose) {name=InPose_PreAimOffsets}
   Output Pose() {layerGroup=Item Override}
  ]}
 ItemFallAdditive {nodes [
   Input Pose() {name=InPoseItemFallAdditive}
   Output Pose() {layerGroup=Item Override}
  ]}
 ItemFallLandAdditive {nodes [
   Input Pose() {name=InPoseItemFallLandAdditive}
   Output Pose() {layerGroup=Item Override}
  ]}
 ItemFallLevel2Override {nodes [
   Input Pose() {name=InPose_FallLevel2}
   Output Pose() {layerGroup=Item Override}
  ]}
 ItemFinalPoseOverride {nodes [
   Input Pose() {name=InFinalPose}
   Output Pose() {layerGroup=Item Override}
  ]}
 ItemFullBodyAimOffsets {nodes [
   Input Pose!() -> (Pose, FullBody_AimOffsetAlpha:double, Aim_Yaw:double, Aim_Pitch:double) {name=InPoseFullBodyAimOffsets}
   Input Pose!() -> (Pose) {name=InPosePreFullBodyAimOffsets}
   Output Pose() {layerGroup=Item Override}
  ]}
 ItemFullBodyOverride {nodes [
   Input Pose() {name=InPoseFullBody}
   Output Pose() {layerGroup=Item Override}
  ]}
 ItemIdleOverride {nodes [
   Input Pose() {name=InPose_Idle}
   Output Pose() {layerGroup=Item Override}
  ]}
 ItemJetPackJumpAdditive {nodes [Output Pose() {layerGroup=Item Override}]}
 ItemJetPackStartAdditive {nodes [Output Pose() {layerGroup=Item Override}]}
 ItemJumpLoopAdditive {nodes [
   Input Pose() {name=InPoseItemJumpLoopAdditive}
   Output Pose() {layerGroup=Item Override}
  ]}
 ItemJumpOverride {nodes [
   Input Pose() {name=InPoseJump}
   Output Pose() {layerGroup=Item Override}
  ]}
 ItemJumpUpAdditive {nodes [
   Input Pose() {name=InPoseItemJumpUpAdditive}
   Output Pose() {layerGroup=Item Override}
  ]}
 ItemPreAimOffsets {nodes [
   Input Pose() {name=InPosePreAimOffsets}
   Output Pose() {layerGroup=Item Override}
  ]}
 ItemPreIK {nodes [
   Input Pose() {name=InPose_PreIK}
   Output Pose() {layerGroup=Item Override}
  ]}
 ItemRightHandIK {nodes [
   Input Pose() {name=InPosePreRightHandIK}
   Input Pose() {name=InPosePostRightHandIK}
   Output Pose() {layerGroup=Item Override}
  ]}
 ItemSkeletalControl {nodes [
   Input Pose()
   Output Pose() {layerGroup=Item Override}
  ]}
 ItemSwimJumpFallLoop {nodes [Output Pose() {layerGroup=Item Override}]}
 ItemSwimJumpSurfaceEnd {nodes [Output Pose() {layerGroup=Item Override}]}
 ItemSwimJumpSurfaceLoop {nodes [Output Pose() {layerGroup=Item Override}]}
 ItemUpperBody {nodes [
   n1 = Apply Additive(
    Base=Use cached pose 'UpperBodySavedPose'()
    Additive="Sequence Player (additive) 'DeployableTurret_NonTargeted_RidingIdle'"() {sequence=&/DeployableTurret_NonTargeted_RidingIdle}
   ) {alphaBoolBlend {blendOption=Linear}}
   n2 = Apply Additive(
    Base=Use cached pose 'UpperBodySavedPose'()
    Additive="Sequence Player (additive) 'DeployableTurret_NonTargeted_SliddingLoop_Additive'"() {sequence=&/DeployableTurret_NonTargeted_SliddingLoop_Additive}
   ) {alphaBoolBlend {blendOption=Linear}}
   n3 = Get bIsRiding() -> (bIsRiding:bool)
   n4 = Get bIsSliding()
   n5 = Get bIsRiding()
   Output Pose(Result=Blend Poses by bool(
     BlendPose_0=Blend Poses by bool(BlendPose_0=@n1, BlendPose_1=@n2, BlendTime_0=0.25, BlendTime_1=0.3, bActiveValue=@n3) {blendType=Linear, transitionType=Inertialization}
     BlendPose_1=Use cached pose 'UpperBodySavedPose'()
     BlendTime_0=0.25
     BlendTime_1=0.3
     bActiveValue=OR Boolean(A=@n4, B=@n5)
    ) {blendType=Linear, transitionType=Inertialization}) {layerGroup=Item Override}
   Save cached pose 'UpperBodySavedPose'(Pose=Input Pose!() -> (Pose, InputParam:Fort Anim Input Adjusted Aim Structure) {name=InPoseUpperBody}) {cacheName=UpperBodySavedPose}
   "Sequence Player (additive) 'DeployableTurret_NonTargeted_Slidding_Additive'"() {bLoopAnimation=false, sequence=&/DeployableTurret_NonTargeted_Slidding_Additive}
  ]}
 ItemZipLineStartAdditive {nodes [Output Pose() {layerGroup=Item Override}]}
]
variables []
