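; Trajectory preview for the deployable turret that adds a warning state to the parent's normal and invalid line and target visuals.
; SetTurretTrajectoryType stores the state, toggling target meshes and swapping spline materials, which are also reapplied when the segment count changes.
; A dynamic warning material receives fade settings and spline endpoints, and the warning target follows the end point.
; The normal target's yaw follows the player's view, and SetTrajectoryValid is overridden empty.
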(event EventBeginPlay
  (|CallParentFunction)
  ;[0.1612 1 0.1056] Initialize parameters on the Warning MID. Parent handles valid and invalid MIDS.
  (Rendering|Material|SetScalarParameterValue (Variables|Default|GetWarningSplineMID) "StartDistance" (Variables|Default|GetSplineFadeStartDistance) :!class /Script/Engine.MaterialInstanceDynamic)
  (Rendering|Material|SetScalarParameterValue (Variables|Default|GetWarningSplineMID) "FadeDistance" (Variables|Default|GetSplineFadeDistance) :!class /Script/Engine.MaterialInstanceDynamic)
  ;[0.1752 0.6541 1] Make sure WarningTarget starts invisible. Parent handles valid/invalid
  (Rendering|SetVisibility (Variables|Default|GetWarningTarget) false false))

(event Trajectory|EventSetTrajectorySpline
  (params (SplinePoints :type Array%20of%20Vectors) (SplineTangents :type Array%20of%20Vectors))
  (bind _returnvalue (Utilities|Array|LastIndex SplinePoints))
  (bind _output (Utilities|Array|Get(aref) SplinePoints _returnvalue))
  ;[0.2329 0.9036 1] SetTrajectorySpline - Update Warning Spline, parent handles valid & invalid
  (|CallParentFunction SplinePoints SplineTangents)
  (if (> (Utilities|Array|Length SplinePoints) 0)
    (Rendering|Material|SetVectorParameterValue (Variables|Default|GetWarningSplineMID) "StartLocation" (Math|Conversions|ToLinearColor(Vector) (Utilities|Array|Get(aref) SplinePoints)) :!class /Script/Engine.MaterialInstanceDynamic)
    (Rendering|Material|SetVectorParameterValue (Variables|Default|GetWarningSplineMID) "EndLocation" (Math|Conversions|ToLinearColor(Vector) _output) :!class /Script/Engine.MaterialInstanceDynamic)
    (Transformation|SetWorldLocation (Variables|Default|GetWarningTarget) _output false true)))

(event Custom|SetTurretTrajectoryType
  (params (TrajectoryType :type E_DeployableTurretTrajectoryType%20Enum :default "NewEnumerator3"))
  ;[0.9836 0.3241 1] Update the spline and target materials depending on which trajectory should be visible
  (if (Utilities|Enum|NotEqual(Enum) TrajectoryType (Variables|Default|GetCurrentTrajectoryType))
    (Variables|Default|SetCurrentTrajectoryType TrajectoryType)
    (CallFunction|UpdateTargetMeshVisibility)
    (CallFunction|UpdateSplineMaterials)))

(event EventSetTrajectoryValid
  (params (bIsTrajectoryValid :type Boolean)))

(event EventUpdateTargetMeshVisibility
  ;[0.9836 0.3241 1] Update the spline and target materials depending on which trajectory should be visible
  (Rendering|SetVisibility (Variables|Default|GetTarget) (Utilities|Enum|Equal(Enum) (Variables|Default|GetCurrentTrajectoryType) "NewEnumerator2") false)
  (Rendering|SetVisibility (Variables|Default|GetInvalidTarget) (Utilities|Enum|Equal(Enum) (Variables|Default|GetCurrentTrajectoryType) "NewEnumerator3") false)
  (Rendering|SetVisibility (Variables|Default|GetWarningTarget) (Utilities|Enum|Equal(Enum) (Variables|Default|GetCurrentTrajectoryType) "NewEnumerator5") false)
  (if (!= (Utilities|Array|Length (Variables|Trajectory|GetSplineMeshComponents)) (Utilities|Array|Length (Variables|Default|GetSpline%20Mesh%20ComponentsLast)))
    (CallFunction|UpdateSplineMaterials)))

(fn UpdateFromTrajectoryOwner
  (|CallParentFunction)
  (bind _asfort_gameplay_ability (Utilities|Casting|CastToFortGameplayAbility (Variables|Default|GetTrajectoryOwner)))
  (bind _outfortplayerpawn (Class|FortGameplayAbility|GetAvatarasValidFortPlayerPawn _asfort_gameplay_ability)
    (:ValidFortPlayerPawn
      (bind (_location _rotation) (Pawn|GetPlayerViewPoint (Fortnite|GetFortPlayerControllerfromActor _outfortplayerpawn)))
      (bind (_rotation_roll _rotation_pitch _rotation_yaw) (Math|Rotator|BreakRotator _rotation))
      (Transformation|SetWorldRotation (Variables|Default|GetTarget) (Math|Rotator|MakeRotator 0.0 0.0 _rotation_yaw) false false))))

(fn UpdateSplineMaterials
  (for _array_element (Variables|Trajectory|GetSplineMeshComponents)
    (Rendering|Material|SetMaterial _array_element 0 (Utilities|Select :Index (Variables|Default|GetCurrentTrajectoryType) :NewEnumerator2 (Variables|VFX|GetSplineMID) :NewEnumerator3 (Variables|VFX|GetInvalidSplineMID) :NewEnumerator5 (Variables|Default|GetWarningSplineMID))))
  (Variables|Default|SetSpline%20Mesh%20ComponentsLast (Variables|Trajectory|GetSplineMeshComponents)))

(fn ConstructionScript
  (|CallParentFunction)
  ;[0.05805 0.4849 1] Create and store warning material instance. Parent handles valid and invalid MIDs
  (bind _returnvalue (Rendering|Material|CreateDynamicMaterialInstance "/Game/Athena/Items/HeldObjects/M_Grenade_TrajectoryLine_01.M_Grenade_TrajectoryLine_01" "None" "None" :!class /Script/Engine.KismetMaterialLibrary))
  (Variables|Default|SetWarningSplineMID _returnvalue))
