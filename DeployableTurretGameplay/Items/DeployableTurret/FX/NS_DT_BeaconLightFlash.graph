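# Niagara system for the deployable turret's beacon light flash, a short-lived burst flare tinted by the User.TeamColor parameter and drawn as both a sprite and a light.
# Stacked scale curves wobble the sprite's aspect and pulse it from zero to peak and back.
# System.DistanceScalar, apparently from a distance LOD remap, drives sprite size, RGB brightness and light radius.
# An older random-flicker size module is disabled.
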NiagaraSystem /DeployableTurretGameplay/Items/DeployableTurret/FX/NS_DT_BeaconLightFlash
dependencies {
 usedDataInterfaces [&NiagaraDataInterfaceCurve, &NiagaraDataInterfaceVector2DCurve]
 usedDynamicInputs [&ClampFloat, &FN_LOD_Distance_Remap]
 usedModules [
  &self:SystemSpawnScript.SystemScriptSource.SystemScriptGraph.NiagaraNodeAssignment_0.SetVariables_#1
  &SystemState
  &self:SystemSpawnScript.SystemScriptSource.SystemScriptGraph.NiagaraNodeAssignment_1.SetVariables_#2
 ]
}
emitters {Flare {
  emitterUpdateScript [
   &EmitterState(Life Cycle Mode=::NewEnumerator1, Loop Behavior=::NewEnumerator1, Loop Duration=1, UseLoopDelay=false)
   &SpawnBurst_Instantaneous(Use Spawn Probability=false, UseLoopCountLimit=false)
  ]
  particleSpawnScript [&InitializeParticle(
    Color=$User.TeamColor
    Lifetime=0.5
    Sprite Size Mode=::NewEnumerator3
    Uniform Sprite Size=500
    UsePositionOffset=false
    Write Material Random=true
   )]
  particleUpdateScript [
   &ParticleState()
   &ScaleSpriteSize(
    Scale Factor=&Multiply_Vector2D_ByFloat(Float=&RandomRangeFloat(Minimum=0.5, Maximum=1, Evaluation Type=::NewEnumerator1), Vector2D=<1 1>)
    Scale Sprite Size Mode=::NewEnumerator2
   ) {enabled=false}
   &ScaleSpriteSize(
    Non-Uniform Curve Sprite Scale=DataInterface() {propertyValues {
      XCurve {keys [
        <at 0 1 RCIM_Cubic>
        {arriveTangent=0.7143, interpMode=RCIM_Cubic, leaveTangent=0.7143, time=0.3, value=0.5}
        {arriveTangent=0.7143, interpMode=RCIM_Cubic, leaveTangent=0.7143, time=0.7, value=1.5}
        <at 1 1 RCIM_Cubic>
       ]}
      YCurve {keys [
        <at 0 1 RCIM_Cubic>
        {arriveTangent=-0.7143, interpMode=RCIM_Cubic, leaveTangent=-0.7143, time=0.3, value=1.5}
        {arriveTangent=-0.7143, interpMode=RCIM_Cubic, leaveTangent=-0.7143, time=0.7, value=0.5}
        <at 1 1 RCIM_Cubic>
       ]}
     }}
    Scale Sprite Size Mode=::NewEnumerator3
   ) {module=ScaleSpriteSize001}
   &ScaleSpriteSize(Scale Sprite Size Mode=::NewEnumerator0, Uniform Scale Factor=&Lerp_Float(A=1, B=3, Alpha=$System.DistanceScalar)) {module=ScaleSpriteSize002}
   &ScaleSpriteSize(Uniform Curve Sprite Scale=DataInterface() {propertyValues {Curve {keys [
        {interpMode=RCIM_Cubic, tangentMode=RCTM_User, time=0, value=0}
        {interpMode=RCIM_Cubic, tangentMode=RCTM_User, time=0.5, value=1}
        {tangentMode=RCTM_User, time=1, value=0}
       ]}}}) {module=ScaleSpriteSize003}
   &FN_Light_Attributes(
    Color=$User.TeamColor
    Create Color Attribute=true
    Create Radius Attribute=true
    Radius=&Lerp_Float(A=300, B=600, Alpha=$System.DistanceScalar)
   )
   &ScaleColor(
    Scale Alpha=0.1
    Scale RGB=&VectorFromFloat(Value=&Lerp_Float(A=0.2, B=8, Alpha=$System.DistanceScalar))
    ScaleA=true
    ScaleRGB=true
   )
  ]
  rendererClasses [&NiagaraSpriteRendererProperties, &NiagaraLightRendererProperties]
  renderers [
   {rendererClass=&NiagaraSpriteRendererProperties}
   {rendererClass=&NiagaraLightRendererProperties, rendererIndex=1}
  ]
 }}
summary {
 systemName=NS_DT_BeaconLightFlash
 emitters [{emitterName=Flare, rendererClasses [&NiagaraSpriteRendererProperties, &NiagaraLightRendererProperties]}]
 userVariables [User.TeamColor {type=LinearColor, description="", defaultValue {struct=LinearColor, value=<rgba 1 0 0 1>}}]
}
