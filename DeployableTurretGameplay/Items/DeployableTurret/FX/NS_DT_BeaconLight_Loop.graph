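# Looping Niagara beacon light for the deployable turret, tinted by User.TeamColor.
# Flare001 pairs a randomly size-modulated sprite with an offset light, while High_glow renders mesh glow with time-of-day-scaled alpha.
# Both burst-spawn particles that outlive their lifetime, using age curves to flash up and settle at steady brightness.
# System.DistanceScalar, alongside an LOD distance remap dependency, adjusts flare alpha and glow width, brightness and material parameters.
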NiagaraSystem /DeployableTurretGameplay/Items/DeployableTurret/FX/NS_DT_BeaconLight_Loop
dependencies {
 usedDataInterfaces [&NiagaraDataInterfaceCurve, &NiagaraDataInterfaceVector2DCurve]
 usedDynamicInputs [&ClampFloat, &FN_LOD_Distance_Remap]
 usedModules [
  &self:SystemSpawnScript.SystemScriptSource.SystemScriptGraph.NiagaraNodeAssignment_0.SetVariables_#1
  &SystemState
  &self:SystemSpawnScript.SystemScriptSource.SystemScriptGraph.NiagaraNodeAssignment_1.SetVariables_#2
 ]
}
emitters {
 Flare001 {
  emitterUpdateScript [
   &EmitterState(
    Inactive Response=::NewEnumerator1
    Life Cycle Mode=::NewEnumerator1
    Loop Behavior=::NewEnumerator1
    Loop Duration Mode=::NewEnumerator1
    UseLoopDelay=false
   )
   &SpawnBurst_Instantaneous(Spawn Time=0.4, Use Spawn Probability=false, UseLoopCountLimit=false)
  ]
  particleSpawnScript [&InitializeParticle(Lifetime=0.5, Sprite Size Mode=::NewEnumerator3, Uniform Sprite Size=300, UsePositionOffset=false, Write Material Random=true)]
  particleUpdateScript [
   &ParticleState(Kill Particles When Lifetime Has Elapsed=false, Let Infinitely Lived Particles Die When Emitter Deactivates=true)
   &ScaleSpriteSize(
    Scale Factor=&Multiply_Vector2D_ByFloat(Float=&RandomRangeFloat(Minimum=0.8, Maximum=1, Evaluation Type=::NewEnumerator1), Vector2D=<1 1>)
    Scale Sprite Size Mode=::NewEnumerator2
   )
   &ScaleSpriteSize(
    Non-Uniform Curve Sprite Scale=DataInterface() {propertyValues {
      XCurve {keys [
        <at 0 1 RCIM_Cubic>
        {arriveTangent=0.7143, interpMode=RCIM_Cubic, leaveTangent=0.7143, time=0.3, value=0.5}
        {arriveTangent=0.7143, interpMode=RCIM_Cubic, leaveTangent=0.7143, time=0.7, value=1.5}
        <at 1 1 RCIM_Cubic>
       ]}
      YCurve {keys [
        <at 0 1 RCIM_Cubic>
        {arriveTangent=-0.7143, interpMode=RCIM_Cubic, leaveTangent=-0.7143, time=0.3, value=1.5}
        {arriveTangent=-0.7143, interpMode=RCIM_Cubic, leaveTangent=-0.7143, time=0.7, value=0.5}
        <at 1 1 RCIM_Cubic>
       ]}
     }}
    Scale Sprite Size Mode=::NewEnumerator3
   ) {enabled=false, module=ScaleSpriteSize001}
   &ScaleSpriteSize(
    Uniform Curve Index=$Particles.Age
    Uniform Curve Sprite Scale=DataInterface() {propertyValues {Curve {keys [<at 0 0 RCIM_Cubic>, {arriveTangent=3.0303, interpMode=RCIM_Cubic, leaveTangent=3.0303, time=0.1, value=2}, <at 0.33 1 RCIM_Cubic>]}}}
   ) {module=ScaleSpriteSize002}
   &FN_Light_Attributes(
    Color=/Niagara/DynamicInputs/LinearColor/MakeLinearColorFromFloat4(Vector:Vector4f=$User.TeamColor)
    Create Color Attribute=true
    Create Position Attribute=true
    Position=&AddVectorToPosition(Position=$Particles.Position, Vector=<85 0 0>)
   )
   &ScaleColor(Scale Mode=::NewEnumerator1, Scale RGBA=$User.TeamColor, ScaleRGBA=true) {module=ScaleColor002}
   &ScaleColor(
    Scale Alpha=&Lerp_Float(A=0.01, B=0.3, Alpha=$System.DistanceScalar)
    Scale RGB=&VectorFromFloat(Value=0.2)
    ScaleA=true
    ScaleRGB=true
   )
   &ScaleColor(
    Scale RGB=&VectorFromFloat(Value=&FloatFromCurve(
      FloatCurve=DataInterface() {propertyValues {Curve {keys [<at 0 0 RCIM_Cubic>, {arriveTangent=3.0303, interpMode=RCIM_Cubic, leaveTangent=3.0303, time=0.1, value=2}, <at 0.33 1 RCIM_Cubic>]}}}
      CurveIndex=$Particles.Age
     ))
    ScaleA=true
    ScaleRGB=true
   ) {module=ScaleColor001}
  ]
  rendererClasses [&NiagaraSpriteRendererProperties, &NiagaraLightRendererProperties]
  renderers [
   {rendererClass=&NiagaraSpriteRendererProperties}
   {rendererClass=&NiagaraLightRendererProperties, rendererIndex=1}
  ]
 }
 High_glow {
  emitterUpdateScript [
   &EmitterState(
    Inactive Response=::NewEnumerator1
    Life Cycle Mode=::NewEnumerator1
    Loop Behavior=::NewEnumerator1
    Loop Duration Mode=::NewEnumerator1
    UseLoopDelay=false
   )
   &SpawnBurst_Instantaneous(Spawn Time=0.4, Use Spawn Probability=false, UseLoopCountLimit=false)
  ]
  particleSpawnScript [
   &InitializeParticle(Mesh Scale=<1.3 1 1>, Mesh Scale Mode=::NewEnumerator4, UsePositionOffset=false, Write Material Random=true)
   &InitialMeshOrientation(Mesh Orientation Mode=::NewEnumerator2, Orientation Vector=<0 0 1>, Use Rotation Vector=true)
   &SystemLocation(MaskSpawnGroup=false, Offset=<0 0 -20>)
   &CameraOffset(Camera Offset Amount=10)
  ]
  particleUpdateScript [
   &ParticleState(Kill Particles When Lifetime Has Elapsed=false, Let Infinitely Lived Particles Die When Emitter Deactivates=true)
   &DynamicMaterialParameters(
    Index 0 Param 1=&Lerp_Float(A=20, B=60, Alpha=$System.DistanceScalar)
    Index 0 Param 2=1
    Index 0 Param 3=1
    Index 0 Param 4=1
    Module.Param0WriteEnabled=true
    Module.Param1WriteEnabled=true
    Module.Param2WriteEnabled=true
    Module.Param3WriteEnabled=true
   )
   &ScaleMeshSize(Scale Factor=&MakeVector(X=&Lerp_Float(A=0.2, B=1, Alpha=$System.DistanceScalar), Y=1, Z=1))
   &ScaleColor(Scale Mode=::NewEnumerator1, Scale RGBA=$User.TeamColor, ScaleRGBA=true) {module=ScaleColor002}
   &ScaleColor(
    Scale Alpha=&FNTimeOfDayScalar(Night Scale=0.5, Morning Scale=0.2, Evening Scale=0.3)
    Scale RGB=&VectorFromFloat(Value=&Lerp_Float(A=0.01, B=0.025, Alpha=$System.DistanceScalar))
    ScaleA=true
    ScaleRGB=true
   )
   &ScaleColor(
    Scale RGB=&VectorFromFloat(Value=&FloatFromCurve(
      FloatCurve=DataInterface() {propertyValues {Curve {keys [<at 0 0 RCIM_Cubic>, {arriveTangent=3.0303, interpMode=RCIM_Cubic, leaveTangent=3.0303, time=0.1, value=2}, <at 0.33 1 RCIM_Cubic>]}}}
      CurveIndex=$Particles.Age
     ))
    ScaleA=true
    ScaleRGB=true
   ) {module=ScaleColor001}
  ]
  rendererClasses [&NiagaraMeshRendererProperties]
  renderers [
   {rendererClass=&NiagaraMeshRendererProperties, rendererIndex=0}
   {rendererClass=&NiagaraMeshRendererProperties, rendererIndex=1}
  ]
 }
}
summary {
 systemName=NS_DT_BeaconLight_Loop
 emitters [
  {emitterName=Flare001, rendererClasses [&NiagaraSpriteRendererProperties, &NiagaraLightRendererProperties]}
  {emitterName=High_glow, rendererClasses [&NiagaraMeshRendererProperties]}
 ]
 userVariables [
  User.TeamColor {type=LinearColor, description="", defaultValue {struct=LinearColor, value=<rgba 1 1 1 1>}}
  User.TeamColor {type=Vector4f, description="", defaultValue {struct=Vector4f, value=<1 1 1 1>}}
 ]
}
