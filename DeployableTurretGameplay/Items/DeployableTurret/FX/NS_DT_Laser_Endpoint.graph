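# Niagara effect for a team-colored flare, named as the endpoint of the deployable turret's laser.
# A single Flare emitter spawns a delayed burst drawn by sprite and light renderers, with particles that persist past their lifetime rather than being killed.
# User.TeamColor tints both sprite and light, a random scale varies sprite size, and an age-driven curve fades in alpha while RGB is dimmed.
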NiagaraSystem /DeployableTurretGameplay/Items/DeployableTurret/FX/NS_DT_Laser_Endpoint
dependencies {usedDataInterfaces [&NiagaraDataInterfaceCurve], usedDynamicInputs [], usedModules [&SystemState]}
emitters {Flare {
  emitterUpdateScript [
   &EmitterState(
    Inactive Response=::NewEnumerator1
    Life Cycle Mode=::NewEnumerator1
    Loop Behavior=::NewEnumerator1
    Loop Duration Mode=::NewEnumerator1
    UseLoopDelay=false
   )
   &SpawnBurst_Instantaneous(Spawn Time=0.3, Use Spawn Probability=false, UseLoopCountLimit=false)
  ]
  particleSpawnScript [&InitializeParticle(Lifetime=5, Sprite Size Mode=::NewEnumerator3, Uniform Sprite Size=60, UsePositionOffset=false, Write Material Random=true)]
  particleUpdateScript [
   &ParticleState(Kill Particles When Lifetime Has Elapsed=false, Let Infinitely Lived Particles Die When Emitter Deactivates=true)
   &ScaleSpriteSize(
    Scale Factor=&Multiply_Vector2D_ByFloat(Float=&RandomRangeFloat(Minimum=0.5, Maximum=1, Evaluation Type=::NewEnumerator1), Vector2D=<1 1>)
    Scale Sprite Size Mode=::NewEnumerator2
   )
   &FN_Light_Attributes(Color=$User.TeamColor, Create Color Attribute=true)
   &ScaleColor(Scale Mode=::NewEnumerator1, Scale RGBA=$User.TeamColor, ScaleRGBA=true) {module=ScaleColor001}
   &ScaleColor(
    Scale Alpha=&FloatFromCurve(
     FloatCurve=DataInterface() {propertyValues {Curve {keys [{interpMode=RCIM_Cubic, tangentMode=RCTM_User, time=0, value=0}, {interpMode=RCIM_Cubic, tangentMode=RCTM_User, time=0.7, value=1}]}}}
     CurveIndex=$Particles.Age
    )
    Scale RGB=&VectorFromFloat(Value=0.2)
    ScaleA=true
    ScaleRGB=true
   )
  ]
  rendererClasses [&NiagaraSpriteRendererProperties, &NiagaraLightRendererProperties]
  renderers [
   {rendererClass=&NiagaraSpriteRendererProperties}
   {rendererClass=&NiagaraLightRendererProperties, rendererIndex=1}
  ]
 }}
summary {
 systemName=NS_DT_Laser_Endpoint
 emitters [{emitterName=Flare, rendererClasses [&NiagaraSpriteRendererProperties, &NiagaraLightRendererProperties]}]
 userVariables [
  User.TeamColor {type=LinearColor, description="", defaultValue {struct=LinearColor, value=<rgba 1 1 1 1>}}
  User.TeamColor {type=Vector4f, description="", defaultValue {struct=Vector4f, value=<0 0 0 0>}}
 ]
}
