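# Niagara muzzle-flash system for the deployable turret, with a bullet tracer and ejected shell casings.
# The mesh tracer aims at User.BeamEnd_0, scaling lifetime and length by shot distance and using a frame-count-aware age to drive travel and fade.
# Shells eject randomly from either side, then fall and spin without collision, with platform-dependent distance culling.
# Flash, spark and light emitters expose no detail here; ExhaustSmoke is disabled.
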NiagaraSystem /DeployableTurretGameplay/Items/DeployableTurret/FX/NS_DT_MF
dependencies {
 usedDataInterfaces [&NiagaraDataInterfaceCurve, &NiagaraDataInterfaceVectorCurve, &NiagaraDataInterfacePlatformSet]
 usedDynamicInputs []
 usedModules [&SystemState]
}
emitters {
 Coreflash_LW {rendererClasses [], renderers []}
 ExhaustSmoke {
  bEnabled=false
  emitterUpdateScript [
   &EmitterState()
   &SpawnBurst_Instantaneous(Spawn Count=&UniformRangedInt(Minimum=2, Maximum=5), Use Spawn Probability=false)
  ]
  particleSpawnScript [
   &InitializeParticle(
    Color Maximum=<rgba 0.5521 0.5521 0.5521 1>
    Color Minimum=<rgba 0.724 0.724 0.724 1>
    Color Mode=::NewEnumerator2
    Lifetime Max=0.4
    Lifetime Min=0.3
    Lifetime Mode=::NewEnumerator1
    Sprite Rotation Mode=::NewEnumerator2
    Sprite Size Mode=::NewEnumerator1
    Uniform Sprite Size Max=80
    Uniform Sprite Size Min=60
    UsePositionOffset=false
    Write Material Random=true
   ) {module=InitializeParticle001}
   /SetVariables(Particles.InitSpriteRotRate=&UniformRangedFloat(Minimum=-3, Maximum=3)) {isSetParameters=true}
   &ShapeLocation(MaskSpawnGroup=false, Non Uniform Scale=<1 2 1>, Override Seed=false, Sphere Radius=2, UseOwnerScale=true)
   &AddVelocityInCone(
    Cone Angle=10
    Module.Use Velocity Falloff On Cone Axis=false
    Velocity Strength=&UniformRangedFloat(Minimum=250, Maximum=350)
   ) {enabled=false}
   &AddVelocity(
    Cone Angle=2
    Module.UseConeFalloff=false
    Module.UseVelDistribution=false
    Override Seed=false
    Velocity Mode=::NewEnumerator2
    Velocity Speed=&RandomRangeFloat(Minimum=300, Maximum=800)
   )
  ]
  particleUpdateScript [
   &ParticleState()
   &SpriteRotationRate(Rotation Rate=&Multiply_Float(
     A=$Particles.InitSpriteRotRate
     B=&FloatFromCurve(FloatCurve=DataInterface() {propertyValues {Curve {keys [<at -0.0046 1.0113 RCIM_Cubic>,<at 1.0011 0.09 RCIM_Cubic>]}}})
    ))
   &Drag(Drag=1.5, Use Linear Drag=true, Use Rotational Drag=false)
   &SolveForcesAndVelocity(Clamp Velocity=false, Limit Acceleration=false)
   &ScaleSpriteSize(Scale Factor=&Vector2DFromFloat(Value=&FloatFromCurve(FloatCurve=DataInterface() {propertyValues {Curve {keys [
          {arriveTangent=2.0031, interpMode=RCIM_Cubic, leaveTangent=2.0031, tangentMode=RCTM_User, time=0.0013, value=0.0367}
          {arriveTangent=0.3654, interpMode=RCIM_Cubic, leaveTangent=0.3654, tangentMode=RCTM_User, time=1.0008, value=1.6042}
         ]}}})))
   /Game/Effects/Niagara/Modules/Materials/MaterialSmoke(
    Erode:NiagaraFloat=&FloatFromCurve(FloatCurve=DataInterface() {propertyValues {
       bOptimizeLUT=false
       bUseLUT=false
       Curve {keys [
         {arriveTangent=-0.0652, interpMode=RCIM_Cubic, leaveTangent=-0.0652, tangentMode=RCTM_User, time=0.11, value=0.97}
         {arriveTangent=-5.7797, interpMode=RCIM_Cubic, leaveTangent=-5.7797, tangentMode=RCTM_User, time=0.39, value=0.66}
         {arriveTangent=-0.342, interpMode=RCIM_Cubic, leaveTangent=-0.342, tangentMode=RCTM_User, time=0.9954, value=-0.74}
        ]}
      }})
    Shading:NiagaraFloat=&FloatFromCurve(FloatCurve=DataInterface() {propertyValues {
       bOptimizeLUT=false
       bUseLUT=false
       Curve {keys [
         <at 0.07 0.5 RCIM_Cubic>
         {arriveTangent=1.2205, interpMode=RCIM_Cubic, leaveTangent=1.2205, time=0.17, value=1.02}
         {interpMode=RCIM_Cubic, tangentMode=RCTM_User, time=0.3895, value=0.89}
         <at 0.8 0 RCIM_Cubic>
        ]}
      }})
   )
   &CameraOffset(Camera Offset Amount=20) {enabled=false}
   &DynamicMaterialParameters(
    Index 0 Param 1=&Add_Float(
     A=&RandomRangeFloat(Minimum=0, Maximum=1)
     B=&FloatFromCurve(FloatCurve=DataInterface() {propertyValues {Curve {keys [<at 0 1>,<at 1 0>]}}})
    )
    Index 0 Param 2=&FloatFromCurve(FloatCurve=DataInterface() {propertyValues {Curve {keys [
         {arriveTangent=0.8564, interpMode=RCIM_Cubic, leaveTangent=0.8564, tangentMode=RCTM_User, time=0.0046, value=-0.9753}
         {arriveTangent=4.2876, interpMode=RCIM_Cubic, leaveTangent=5.5268, tangentMode=RCTM_User, time=1, value=1}
        ]}}})
    Index 0 Param 3=&FloatFromCurve(
     FloatCurve=DataInterface() {propertyValues {Curve {keys [
         {arriveTangent=-5.8322, interpMode=RCIM_Cubic, leaveTangent=-5.8322, tangentMode=RCTM_User, time=0, value=1}
         {tangentMode=RCTM_User, time=0.35, value=0}
        ]}}}
     Scale Curve=2
    )
    Index 0 Param 4=&FloatFromCurve(FloatCurve=DataInterface() {propertyValues {Curve {keys [{arriveTangent=-3.2497, interpMode=RCIM_Cubic, leaveTangent=-3.2497, tangentMode=RCTM_User, time=0, value=1}, <at 1 0 RCIM_Cubic>]}}})
    Index 2 Param 1=&FloatFromCurve(FloatCurve=DataInterface() {propertyValues {Curve {keys [
         {interpMode=RCIM_Cubic, leaveTangent=3.0943, tangentMode=RCTM_Break, time=0, value=0}
         {interpMode=RCIM_Cubic, tangentMode=RCTM_Break, time=1, value=1}
        ]}}})
    Index 2 Param 2=&FloatFromCurve(
     FloatCurve=DataInterface() {propertyValues {Curve {keys [
         {interpMode=RCIM_Cubic, leaveTangentWeight=0.397, tangentMode=RCTM_Break, time=0, value=0.7}
         {arriveTangentWeight=0.4342, interpMode=RCIM_Cubic, tangentMode=RCTM_Break, time=1, value=1}
        ]}}}
     Scale Curve=&RandomRangeFloat(Minimum=0, Maximum=1)
    )
    Index 2 Param 3=&FloatFromCurve(
     FloatCurve=DataInterface() {propertyValues {Curve {keys [
         {interpMode=RCIM_Cubic, leaveTangentWeight=0.3345, tangentMode=RCTM_Break, time=0, value=0}
         {arriveTangentWeight=0.3308, interpMode=RCIM_Cubic, tangentMode=RCTM_Break, time=1, value=1}
        ]}}}
     Scale Curve=5
    )
    Index 2 Param 4=&FloatFromCurve(FloatCurve=DataInterface() {propertyValues {Curve {keys [
         {interpMode=RCIM_Cubic, leaveTangentWeight=0.4044, tangentMode=RCTM_Break, time=0, value=0.8}
         {arriveTangent=0.4016, arriveTangentWeight=0.382, interpMode=RCIM_Cubic, tangentMode=RCTM_Break, time=1, value=1}
        ]}}})
    Module.Param0WriteEnabled=true
    Module.Param0WriteEnabled2=true
    Module.Param1WriteEnabled=true
    Module.Param1WriteEnabled2=true
    Module.Param2WriteEnabled=true
    Module.Param2WriteEnabled2=true
    Module.Param3WriteEnabled=true
    Module.Param3WriteEnabled2=true
    Write Parameter Index 2=true
   )
   &ScaleColor(
    Scale Alpha=&FloatFromCurve(
     FloatCurve=DataInterface() {propertyValues {Curve {keys [
         {arriveTangent=1.4377, interpMode=RCIM_Cubic, leaveTangent=1.4377, tangentMode=RCTM_User, time=0, value=0}
         {arriveTangent=0.006, interpMode=RCIM_Cubic, time=0.15, value=1}
         <at 1 0>
        ]}}}
     Scale Curve=0.2
    )
    Scale RGB=/Niagara/DynamicInputs/ScaleByCurve/ScaleVectorByCurve(
     InitialValue:Vector3f=<0.8 0.7 0.6>
     ScaleCurve:NiagaraDataInterfaceVectorCurve=DataInterface() {propertyValues {XCurve {keys [<at 0 1>,<at 1 0>]}, YCurve {keys [<at 0 1>,<at 1 0>]}, ZCurve {keys [<at 0 1>,<at 1 0>]}}}
     CurveIndex:NiagaraFloat=$Particles.NormalizedAge
    )
    ScaleA=false
    ScaleRGB=true
   )
  ]
  rendererClasses [&NiagaraSpriteRendererProperties]
  renderers [{rendererClass=&NiagaraSpriteRendererProperties}]
 }
 FN_Bullet_Tracer {
  emitterSpawnScript [/SetVariables(Emitter.BeamEnd:Vector3f=$User.BeamEnd_0, Emitter.DebugTimeMultiplier=1) {isSetParameters=true}]
  emitterUpdateScript [
   &EmitterState()
   &SpawnBurst_Instantaneous(Spawn Count=1, Use Spawn Probability=false)
  ]
  particleSpawnScript [
   /SetVariables(
    Particles.BeamTile:NiagaraFloat=&Divide_Float(
     A=/Niagara/DynamicInputs/TypeConversions/VectorLengthSafe(Vector:Vector3f=&Subtract_Vector(A=Linked() {linkedVariable {name=Engine.Owner.Position, type=NiagaraPosition}}, B=$Emitter.BeamEnd))
     B=500
    )
    Particles.FireDirection:Vector3f=/Niagara/DynamicInputs/Helpers/SafeNormalizeVector(
     Vector To Normalize:Vector3f=&Subtract_Vector(A=$Emitter.BeamEnd, B=Linked() {linkedVariable {name=Engine.Owner.Position, type=NiagaraPosition}})
     Fallback Vector:Vector3f=$Engine.Owner.SystemXAxis
    )
   ) {isSetParameters=true}
   /SetVariables(
    Particles.BeamDistanceScalar:NiagaraFloat=&Lerp_Float(
     A=0.35
     B=1
     Alpha=&FloatFromCurve(
      FloatCurve=DataInterface() {propertyValues {Curve {keys [
          {arriveTangent=0.0002, interpMode=RCIM_Cubic, leaveTangent=0.0002, tangentMode=RCTM_User, time=0, value=0}
          {arriveTangent=0.0001, interpMode=RCIM_Cubic, leaveTangent=0.0001, tangentMode=RCTM_User, time=2500, value=0.35}
          <at 15000 0.8>
          {interpMode=RCIM_Cubic, tangentMode=RCTM_User, time=35000, value=1}
         ]}}}
      CurveIndex=&VectorLengthSafe(Vector:Vector3f=&Subtract_Vector(A=Linked() {linkedVariable {name=Engine.Owner.Position, type=NiagaraPosition}}, B=$Emitter.BeamEnd))
     )
    )
    Particles.LODDistanceRemap:NiagaraFloat=&FloatFromCurve(FloatCurve=DataInterface() {propertyValues {Curve {keys [<at 500 0>,<at 2500 1>]}}}, CurveIndex=$Engine.Owner.LODDistance)
   ) {isSetParameters=true}
   &InitializeParticle(
    Color=<rgba 1.2 0.4252 0.0313 0.95>
    Lifetime=&Multiply_Float(
     A=&Multiply_Float(
      A=&ClampFloat(Float=$Particles.BeamDistanceScalar, Min=0.5)
      B=&RandomRangeFloat(Minimum=0.7, Maximum=0.8)
     )
     B=$Emitter.DebugTimeMultiplier
    )
    Mesh Scale Mode=::NewEnumerator3
    UsePositionOffset=false
    Write Material Random=true
   )
   &DynamicMaterialParameters(
    Index 0 Param 4=&Multiply_Float(A=$Particles.BeamTile, B=&Lerp_Float(A=0.7, B=1.22, Alpha=$Particles.LODDistanceRemap))
    Index 1 Param 1=&RandomRangeFloat(Minimum=0.85, Maximum=1.05)
    Index 1 Param 2=&RandomRangeFloat(Minimum=0, Maximum=1)
    Module.Param0WriteEnabled=false
    Module.Param0WriteEnabled1=false
    Module.Param1WriteEnabled=false
    Module.Param1WriteEnabled1=true
    Module.Param2WriteEnabled=false
    Module.Param2WriteEnabled1=false
    Module.Param3WriteEnabled=true
    Module.Param3WriteEnabled1=false
    Write Parameter Index 1=true
   ) {module=DynamicMaterialParameters001}
   /Niagara/Modules/Update/Orientation/V2/OrientMeshToVector(Look At Direction:Vector3f=$Particles.FireDirection, Override Quat=false, UseOptionalSideAxis=false, UseOptionalUpAxis=false)
   /SetVariables(Particles.BulletDynParams:Vector4f=<1 1 1 1>) {isSetParameters=true}
  ]
  particleUpdateScript [
   &ParticleState()
   /Niagara/Modules/Update/Utility/FrameCounter()
   /SetVariables(Particles.FrameLimitedAge:NiagaraFloat=&self:FN_Bullet_Tracer_1.NiagaraScratchPadContainer_1.ScratchDynamicInput_6(
     MaxDuration:NiagaraFloat=0.5
     MinDuration:NiagaraFloat=0.1
     CurrentFrame:NiagaraFloat=&MakeFloatFromInt(Int=$Particles.FrameCounter.Counter)
     MaxFrames:NiagaraFloat=&MakeFloatFromInt(Int=8)
     DurationMultiplier:NiagaraFloat=$Particles.BeamDistanceScalar
     DebugTimeMultiplier:NiagaraFloat=$Emitter.DebugTimeMultiplier
    )) {isSetParameters=true}
   /SetVariables(Particles.Travel:NiagaraFloat=&self:FN_Bullet_Tracer_1.NiagaraScratchPadContainer_1.RemapRange_7(
     Value:NiagaraFloat=$Particles.FrameLimitedAge
     Input Min:NiagaraFloat=0
     Input Max:NiagaraFloat=0.5
     Output Min:NiagaraFloat=0
     Output Max:NiagaraFloat=1
    )) {isSetParameters=true}
   /SetVariables(
    Particles.BulletPosition:Vector3f=&Lerp_Vector(
     A=&Add_Vector(
      A=Linked() {linkedVariable {name=Particles.Initial.Position, type=NiagaraPosition}}
      B=&Multiply_VectorByFloat(Vector=$Particles.FireDirection, Float=350)
     )
     B=$Emitter.BeamEnd
     Alpha=$Particles.Travel
    )
    Particles.BulletVisbility:NiagaraInt32=&MakeCustomIntFromBool(
     Bool=&Boolean_AndOperation(
      Bool1=&SetBoolByFloatComparison(A=$Particles.FrameLimitedAge, B=0.55, Comparison Type=::NewEnumerator6)
      Bool2=&SetBoolByFloatComparison(A=$Particles.BeamTile, B=1)
     )
     True INT=0
     False INT=1
    )
   ) {isSetParameters=true}
   /SetVariables(Particles.BulletSize:Vector3f=&Multiply_VectorByFloat(
     Vector=&Lerp_Vector(A=<2.5 1.6 1.6>, B=<4 2.5 2.5>, Alpha=$Particles.LODDistanceRemap)
     Float=&FloatFromCurve(
      FloatCurve=DataInterface() {propertyValues {Curve {keys [{arriveTangent=32.8701, time=0, value=0}, <at 0.05 1>]}}}
      CurveIndex=$Particles.FrameLimitedAge
     )
    )) {isSetParameters=true}
   &ScaleMeshSize(Scale Factor=&MakeVector(
     X=&Lerp_Float(
      A=&Multiply_Float(A=$Particles.Travel, B=$Particles.BeamTile)
      B=100
      Alpha=&FloatFromCurve(
       FloatCurve=DataInterface() {propertyValues {Curve {keys [
           {interpMode=RCIM_Cubic, tangentMode=RCTM_Break, time=0.65, value=0}
           {arriveTangent=9.0989, interpMode=RCIM_Cubic, tangentMode=RCTM_Break, time=1, value=1}
          ]}}}
       CurveIndex=$System.BeamDistanceNormalized
      )
     )
     Y=1
     Z=1
    )) {module=ScaleMeshSize002}
   &ScaleMeshSize(Scale Factor=&MakeVector(X=&Multiply_Float(A=$Particles.Travel, B=$Particles.BeamTile), Y=1, Z=1)) {enabled=false}
   &DynamicMaterialParameters(
    Index 0 Param 1=&FloatFromCurve(FloatCurve=DataInterface() {propertyValues {Curve {keys [
         {arriveTangent=-0.4135, interpMode=RCIM_Cubic, leaveTangent=-0.4135, tangentMode=RCTM_User, time=0, value=1}
         {arriveTangent=-1.3177, interpMode=RCIM_Cubic, leaveTangent=-1.3177, tangentMode=RCTM_User, time=1, value=0}
        ]}}})
    Index 0 Param 2=&FloatFromCurve(
     FloatCurve=DataInterface() {propertyValues {Curve {keys [
         {arriveTangent=5.2627, interpMode=RCIM_Cubic, leaveTangent=5.2627, tangentMode=RCTM_User, time=0, value=0}
         {arriveTangent=0.8931, interpMode=RCIM_Cubic, leaveTangent=0.8931, tangentMode=RCTM_User, time=0.2287, value=0.7}
         {arriveTangentWeight=0.2571, interpMode=RCIM_Cubic, tangentMode=RCTM_User, time=1, value=1}
        ]}}}
     Scale Curve=&Multiply_Float(
      A=&Lerp_Float(A=0.5, B=1, Alpha=$Particles.BeamDistanceScalar)
      B=&RandomRangeFloat(Minimum=-12.5, Maximum=-11)
     )
    )
    Index 0 Param 3=&FloatFromCurve(
     FloatCurve=DataInterface() {propertyValues {Curve {keys [
         {arriveTangent=-0.9882, time=0.2, value=1}
         {arriveTangent=-1.2626, interpMode=RCIM_Cubic, leaveTangent=-1.2626, tangentMode=RCTM_User, time=0.5, value=0.6}
         {arriveTangent=-10.4271, interpMode=RCIM_Cubic, leaveTangent=-10.4271, tangentMode=RCTM_User, time=0.6, value=0}
        ]}}}
     CurveIndex=$Particles.FrameLimitedAge
    )
    Index 0 Param 4=&Multiply_Float(
     A=&MakeFloatFromVector(Vector=$Particles.Scale, Channel=::NewEnumerator0)
     B=&Lerp_Float(A=0.8, B=1.35, Alpha=$Particles.LODDistanceRemap)
    )
    Index 1 Param 1=&FloatFromCurve(
     FloatCurve=DataInterface() {propertyValues {Curve {keys [
         {arriveTangent=-0.8182, interpMode=RCIM_Cubic, leaveTangent=-0.8182, tangentMode=RCTM_User, time=0, value=0.7}
         {arriveTangent=2.4059, interpMode=RCIM_Cubic, leaveTangent=9.91, tangentMode=RCTM_Break, time=0.5, value=0.87}
         {arriveTangent=0.8994, interpMode=RCIM_Cubic, leaveTangent=0.8994, tangentMode=RCTM_User, time=0.5714, value=1.214}
         {arriveTangent=0.8128, interpMode=RCIM_Cubic, leaveTangent=0.8128, tangentMode=RCTM_User, time=1, value=1.5}
        ]}}}
     Scale Curve=&Lerp_Float(A=0.47, B=0.72, Alpha=$Particles.LODDistanceRemap)
    )
    Module.Param0WriteEnabled=true
    Module.Param0WriteEnabled1=true
    Module.Param1WriteEnabled=true
    Module.Param1WriteEnabled1=false
    Module.Param2WriteEnabled=true
    Module.Param2WriteEnabled1=false
    Module.Param3WriteEnabled=false
    Module.Param3WriteEnabled1=false
    Write Parameter Index 1=true
   )
   &ScaleColor(
    Scale Alpha=&FloatFromCurve(FloatCurve=DataInterface() {propertyValues {Curve {keys [
         {arriveTangent=-0.8527, interpMode=RCIM_Cubic, leaveTangent=-0.8527, tangentMode=RCTM_User, time=0.5, value=1}
         {arriveTangent=-1.0784, interpMode=RCIM_Cubic, leaveTangent=-1.0784, tangentMode=RCTM_User, time=0.52, value=0.6}
         {arriveTangent=-2.2297, interpMode=RCIM_Cubic, leaveTangent=-2.2297, tangentMode=RCTM_User, time=1, value=0}
        ]}}})
    Scale RGB=&VectorFromCurve(
     VectorCurve=DataInterface() {propertyValues {
       XCurve {keys [{arriveTangent=-4.22, interpMode=RCIM_Cubic, time=0.35, value=1}, {interpMode=RCIM_Cubic, leaveTangent=-7.2396, time=0.6, value=0}]}
       YCurve {keys [{arriveTangent=-8.4239, interpMode=RCIM_Cubic, time=0.35, value=1}, <at 0.6 0 RCIM_Cubic>]}
       ZCurve {keys [{arriveTangent=-8.2416, interpMode=RCIM_Cubic, time=0.35, value=1}, <at 0.6 0 RCIM_Cubic>]}
      }}
     CurveIndex=$Particles.FrameLimitedAge
     Scale Curve=&VectorFromFloat(Value=&FNTimeOfDayScalar(Night Scale=1, Morning Scale=1, Day Scale=1.5, Evening Scale=1))
    )
    ScaleA=true
    ScaleRGB=true
   )
  ]
  rendererClasses [&NiagaraMeshRendererProperties]
  renderers [
   {rendererClass=&NiagaraMeshRendererProperties, rendererIndex=0}
   {rendererClass=&NiagaraMeshRendererProperties, rendererIndex=1}
  ]
 }
 FlashStreaks_LW {rendererClasses [], renderers []}
 Light {rendererClasses [], renderers []}
 MuzzleFLash_LW {rendererClasses [], renderers []}
 Shell_No_Collide {
  emitterUpdateScript [
   &EmitterState(
    Enable Distance Culling=true
    Max Distance Response=::KillAfterParticlesFinish
    MaxDistance:NiagaraFloat=&FN_PickFloatDependOnPlatform(TargetPlatform=DataInterface() {propertyValues {Platforms {qualityLevelMask=-24}}}, TargetPlatformActive=10000, TargetPlatformNoActive=4000)
    Module.bUseMaxDistance=true
    Module.bUseMinDistance=false
    Scalability Mode=::NewEnumerator1
   )
   &SpawnBurst_Instantaneous(Use Spawn Probability=false, UseLoopCountLimit=false)
  ]
  particleSpawnScript [
   /SetVariables(Particles.VelRand:NiagaraFloat=&Multiply_Float(A=&RandomRangeFloat(Minimum=150, Maximum=200), B=1)) {isSetParameters=true}
   /SetVariables(Particles.RandomBool:NiagaraBool=&RandomBool()) {isSetParameters=true}
   &InitializeParticle(
    Color=<rgba 1 0.0335 0 1>
    Lifetime Max=0.75
    Lifetime Min=0.5
    Lifetime Mode=::NewEnumerator1
    Mesh Scale=<1 1.3 1.3>
    Mesh Scale Mode=::NewEnumerator4
    Position Offset=&LerpPosition(
     A=<-67 12 0>
     B=<-67 -12 0>
     Alpha=&MakeCustomFloatFromBool(Bool=$Particles.RandomBool, True Float=1, False Float=0)
    )
    Position Offset Coordinate Space=::Local
    UsePositionOffset=true
    Write Material Random=true
   )
   &InitialMeshOrientation(Mesh Orientation Mode=::NewEnumerator2, Orientation Coordinate Space=::Local, Use Rotation Vector=false)
   &AddVelocity(
    Velocity=&Lerp_Vector(
     A=&RandomRangeVector(Minimum=<0.2 0.5 0.5>, Maximum=<-1 1 1>)
     B=&RandomRangeVector(Minimum=<0.2 -0.5 0.5>, Maximum=<-1 -1 1>)
     Alpha=&MakeCustomFloatFromBool(Bool=$Particles.RandomBool, True Float=1, False Float=0)
    )
    Velocity Speed Scale=&RandomRangeFloat(Minimum=100, Maximum=300)
   ) {module=AddVelocity002}
  ]
  particleUpdateScript [
   &GravityForce(Gravity=&RandomRangeVector(Minimum=<0 0 -1000>, Maximum=<0 0 -900>))
   &SolveForcesAndVelocity(Clamp Velocity=false, Limit Acceleration=false)
   &ParticleState()
   &UpdateMeshOrientation(
    Rotation Rate=&Multiply_Float(
     A=&RandomRangeFloat(Minimum=2, Maximum=7)
     B=&FloatFromCurve(FloatCurve=DataInterface() {propertyValues {Curve {keys [
          {interpMode=RCIM_Cubic, tangentMode=RCTM_Break, time=0, value=0}
          {interpMode=RCIM_Cubic, tangentMode=RCTM_Break, time=0.2, value=1}
          <at 1 0.1>
         ]}}})
    )
    Rotation Vector=<1 1 0.5>
   )
   &ScaleMeshSize(Scale Factor=&VectorFromFloat(Value=&FloatFromCurve(FloatCurve=DataInterface() {propertyValues {Curve {keys [<at 0.8 1>,<at 1 0.2>]}}})))
  ]
  rendererClasses [&NiagaraMeshRendererProperties]
  renderers [{rendererClass=&NiagaraMeshRendererProperties, rendererIndex=0}]
 }
 Spark002 {rendererClasses [], renderers []}
}
summary {
 systemName=NS_DT_MF
 emitters [
  {bEnabled=false, emitterName=ExhaustSmoke, rendererClasses [&NiagaraSpriteRendererProperties]}
  {emitterName=FN_Bullet_Tracer, rendererClasses [&NiagaraMeshRendererProperties]}
  {emitterName=Shell_No_Collide, rendererClasses [&NiagaraMeshRendererProperties]}
  {emitterName=Coreflash_LW, rendererClasses []}
  {emitterName=Light, rendererClasses []}
  {emitterName=FlashStreaks_LW, rendererClasses []}
  {emitterName=MuzzleFLash_LW, rendererClasses []}
  {emitterName=Spark002, rendererClasses []}
 ]
 userVariables [
  User.BeamEnd_0 {type=Vector3f, description="", defaultValue {struct=Vector3f, value=<0 0 2000>}}
  User.MuzzleLeft {type=NiagaraFloat, defaultValue=0, description=""}
  User.RibbonIDOffset {type=NiagaraInt32}
 ]
}
