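# Physics asset for the briefcase gadget mesh of the deployable turret, giving the case simple collision and constrained physical motion.
# A box body covers dyn_case, and a tiny capsule body sits on dyn_case_mstr as the other end of the joint.
# One constraint links the two, locking twist and swing1 and leaving a limited swing2, so the case swings about a single axis like a hinge.
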PhysicsAsset /DeployableTurretGameplay/Items/DeployableTurret/Meshes/DeployableTurret_Briefcase_Gadget_Physics
bodies {
 dyn_case [{shapeType=Box, extentX=32, extentY=12, extentZ=42, shapeName=dyn_case_box, center=<16.5 0 0>, rotation=<rot 0.0001 0 90>}]
 dyn_case_mstr [{shapeType=Capsule, length=0.505, radius=0.505, shapeName=dyn_case_mstr_capsule}]
}
constraints [{
  bone1Name=dyn_case
  bone2Name=dyn_case_mstr
  swing1LimitDegrees=0
  swing1Motion=Locked
  swing2LimitDegrees=28
  twistLimitDegrees=0
  twistMotion=Locked
 }]
