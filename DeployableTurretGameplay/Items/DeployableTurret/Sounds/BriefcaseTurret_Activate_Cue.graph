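# Sound cue that plays the activation sound for the briefcase deployable turret.
# After a fixed half-second delay, it picks one of three equally weighted activation waves with slight random pitch variation.
# It uses the turret spawn attenuation and the gadget items sound class at reduced volume.
# Overridden concurrency allows one instance per owner, stopping the oldest when a new one starts.
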SoundCue /DeployableTurretGameplay/Items/DeployableTurret/Sounds/BriefcaseTurret_Activate_Cue
root=Delay_0
nodes [
 {id=Delay_0, children [Modulator_0], properties {delayMax=0.5, delayMin=0.5}}
 {id=Modulator_0, children [Random_0], properties {pitchMax=1.1, pitchMin=0.9}}
 {id=Random_0, children [WavePlayer_0,WavePlayer_1,WavePlayer_2], properties {weights [1,1,1]}}
 {id=WavePlayer_0, properties {soundWaveAssetPtr=&/BriefcaseTurret_Activate_01}}
 {id=WavePlayer_1, properties {soundWaveAssetPtr=&/BriefcaseTurret_Activate_02}}
 {id=WavePlayer_2, properties {soundWaveAssetPtr=&/BriefcaseTurret_Activate_03}}
]
properties {
 attenuationSettings=/DeployableTurretGameplay/Items/DeployableTurret/Sounds/Settings/BriefcaseTurret_Spawn_Att
 bOverrideConcurrency=true
 maxDistance=2400
 soundClassObject=&Fort_Gadgets_Items
 volumeMultiplier=0.6
 concurrencyOverrides {bLimitToOwner=true, maxCount=1, resolutionRule=StopOldest}
}
