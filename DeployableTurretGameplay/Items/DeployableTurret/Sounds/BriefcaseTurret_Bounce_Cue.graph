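# Sound cue for the bounce of the deployable briefcase turret, in the gadget items sound class.
# A mixer layers one of six randomly chosen bounce waves, at full level, with a fixed B-variant bounce at lower level, each layer with its own random pitch and volume modulation.
# It borrows the turret's Sight attenuation asset and overrides concurrency to allow one instance per owner, preventing new plays while one is active.
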SoundCue /DeployableTurretGameplay/Items/DeployableTurret/Sounds/BriefcaseTurret_Bounce_Cue
root=Mixer_0
nodes [
 {id=Mixer_0, children [Modulator_1,Modulator_0], properties {inputVolume [0.6,1]}}
 {id=Modulator_0, children [Random_0], properties {pitchMax=1.1, pitchMin=0.9, volumeMax=1.1, volumeMin=0.9}}
 {id=Modulator_1, children [WavePlayer_7], properties {pitchMax=1.2, pitchMin=0.8, volumeMax=1.1, volumeMin=0.9}}
 {id=Random_0, children [WavePlayer_0,WavePlayer_2,WavePlayer_3,WavePlayer_4,WavePlayer_5,WavePlayer_6], properties {weights [1,1,1,1,1,1]}}
 {id=WavePlayer_0, properties {soundWaveAssetPtr=&/BriefcaseTurret_Bounce_01}}
 {id=WavePlayer_2, properties {soundWaveAssetPtr=&/BriefcaseTurret_Bounce_02}}
 {id=WavePlayer_3, properties {soundWaveAssetPtr=&/BriefcaseTurret_Bounce_03}}
 {id=WavePlayer_4, properties {soundWaveAssetPtr=&/BriefcaseTurret_Bounce_04}}
 {id=WavePlayer_5, properties {soundWaveAssetPtr=&/BriefcaseTurret_Bounce_05}}
 {id=WavePlayer_6, properties {soundWaveAssetPtr=&/BriefcaseTurret_Bounce_06}}
 {id=WavePlayer_7, properties {soundWaveAssetPtr=&/BriefcaseTurret_Bounce_B_01}}
]
properties {
 attenuationSettings=/DeployableTurretGameplay/Items/DeployableTurret/Sounds/Settings/BriefcaseTurret_Sight_Att
 bOverrideConcurrency=true
 maxDistance=4100
 soundClassObject=&Fort_Gadgets_Items
 volumeMultiplier=0.3
 attenuationOverrides {
  bEnableFocusInterpolation=true
  bEnableListenerFocus=true
  falloffDistance=1200
  nonFocusVolumeAttenuation=0.5
  stereoSpread=500
  attenuationShapeExtents {x=0}
 }
 concurrencyOverrides {bLimitToOwner=true, maxCount=1, resolutionRule=PreventNew}
}
