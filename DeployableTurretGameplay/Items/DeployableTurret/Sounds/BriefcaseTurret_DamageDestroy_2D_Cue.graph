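# Sound cue for the 2D layer of the deployable briefcase turret's destruction by damage, in the gadget items sound class.
# A random node picks one of three equally weighted distant destroy waves, and a modulator varies its pitch and volume upward slightly.
# Overridden attenuation turns off both distance falloff and spatialization, with the cue's overall volume reduced.
# An object-impacts concurrency set is assigned to limit overlapping plays.
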SoundCue /DeployableTurretGameplay/Items/DeployableTurret/Sounds/BriefcaseTurret_DamageDestroy_2D_Cue
root=Modulator_1
nodes [
 {id=Modulator_1, children [Random_1], properties {pitchMax=1.2, pitchMin=1, volumeMax=1.2, volumeMin=1}}
 {id=Random_1, children [WavePlayer_0,WavePlayer_4,WavePlayer_5], properties {weights [1,1,1]}}
 {
  id=WavePlayer_0
  properties {soundWaveAssetPtr=&/BriefcaseTurret_DamageDestroy_Distant_01}
 }
 {
  id=WavePlayer_4
  properties {soundWaveAssetPtr=&/BriefcaseTurret_DamageDestroy_Distant_02}
 }
 {
  id=WavePlayer_5
  properties {soundWaveAssetPtr=&/BriefcaseTurret_DamageDestroy_Distant_03}
 }
]
properties {
 bOverrideAttenuation=true
 maxDistance=2097152
 soundClassObject=&Fort_Gadgets_Items
 volumeMultiplier=0.4
 attenuationOverrides {
  bAttenuate=false
  bSpatialize=false
  distanceAlgorithm=Logarithmic
  falloffDistance=4000
  occlusionInterpolationTime=1
  occlusionLowPassFilterFrequency=10000
  occlusionTraceChannel=ECC_GameTraceChannel11
  stereoSpread=1200
  attenuationShapeExtents {x=900}
 }
 concurrencyOverrides {maxCount=3, resolutionRule=StopOldest}
 concurrencySet [&CONC_Object_Impacts_FourMax]
}
