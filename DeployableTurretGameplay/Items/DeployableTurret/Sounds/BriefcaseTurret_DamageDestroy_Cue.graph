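# Sound cue for the deployable briefcase turret's damage-destroy sound, routed through the gadgets items sound class.
# It randomly picks one of three equally weighted destruction waves, then a modulator varies pitch and volume slightly upward.
# Overridden attenuation uses logarithmic falloff with occlusion low-pass filtering, and it references a shared object-impacts concurrency set to limit overlapping plays.
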SoundCue /DeployableTurretGameplay/Items/DeployableTurret/Sounds/BriefcaseTurret_DamageDestroy_Cue
root=Modulator_1
nodes [
 {id=Modulator_1, children [Random_0], properties {pitchMax=1.2, pitchMin=1, volumeMax=1.2, volumeMin=1}}
 {id=Random_0, children [WavePlayer_1,WavePlayer_2,WavePlayer_3], properties {weights [1,1,1]}}
 {id=WavePlayer_1, properties {soundWaveAssetPtr=&/BriefcaseTurret_DamageDestroy_01}}
 {id=WavePlayer_2, properties {soundWaveAssetPtr=&/BriefcaseTurret_DamageDestroy_02}}
 {id=WavePlayer_3, properties {soundWaveAssetPtr=&/BriefcaseTurret_DamageDestroy_03}}
]
properties {
 bOverrideAttenuation=true
 maxDistance=4900
 soundClassObject=&Fort_Gadgets_Items
 volumeMultiplier=0.9
 attenuationOverrides {
  bEnableOcclusion=true
  distanceAlgorithm=Logarithmic
  falloffDistance=4000
  occlusionInterpolationTime=1
  occlusionLowPassFilterFrequency=10000
  occlusionTraceChannel=ECC_GameTraceChannel11
  stereoSpread=1200
  attenuationShapeExtents {x=900}
 }
 concurrencyOverrides {maxCount=3, resolutionRule=StopOldest}
 concurrencySet [&CONC_Object_Impacts_FourMax]
}
