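# Sound cue for damage impacts on the deployable briefcase turret, classed and attenuated as a bullet impact.
# It layers a random pick of six turret-specific impact sounds over a quieter random pick of four reused Security Turret impacts.
# Each layer gets its own random pitch and volume modulation before mixing, and the mix plays after a short fixed delay.
# A concurrency override blocks new instances rather than cutting off playing ones when the limit is reached.
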SoundCue /DeployableTurretGameplay/Items/DeployableTurret/Sounds/BriefcaseTurret_DamageImpact_Cue
root=Delay_0
nodes [
 {id=Delay_0, children [Mixer_0], properties {delayMax=0.05, delayMin=0.05}}
 {id=Mixer_0, children [Modulator_1,Modulator_0], properties {inputVolume [1,0.6]}}
 {id=Modulator_0, children [Random_3], properties {pitchMax=1.1, pitchMin=0.9, volumeMax=1.1, volumeMin=0.9}}
 {id=Modulator_1, children [Random_0], properties {pitchMax=1.2, pitchMin=1, volumeMax=1.1, volumeMin=0.9}}
 {id=Random_0, children [WavePlayer_0,WavePlayer_1,WavePlayer_2,WavePlayer_3,WavePlayer_4,WavePlayer_5], properties {weights [1,1,1,1,1,1]}}
 {id=Random_3, children [WavePlayer_14, WavePlayer_15, WavePlayer_16, WavePlayer_17], properties {weights [1,1,1,1]}}
 {id=WavePlayer_0, properties {soundWaveAssetPtr=&/BriefcaseTurret_DamageImpact_01}}
 {id=WavePlayer_1, properties {soundWaveAssetPtr=&/BriefcaseTurret_DamageImpact_02}}
 {id=WavePlayer_14, properties {soundWaveAssetPtr=/Game/Athena/Sounds/Gadgets/SecurityTurret/Security_Turret_Impact_01}}
 {id=WavePlayer_15, properties {soundWaveAssetPtr=/Game/Athena/Sounds/Gadgets/SecurityTurret/Security_Turret_Impact_03}}
 {id=WavePlayer_16, properties {soundWaveAssetPtr=/Game/Athena/Sounds/Gadgets/SecurityTurret/Security_Turret_Impact_04}}
 {id=WavePlayer_17, properties {soundWaveAssetPtr=/Game/Athena/Sounds/Gadgets/SecurityTurret/Security_Turret_Impact_05}}
 {id=WavePlayer_2, properties {soundWaveAssetPtr=&/BriefcaseTurret_DamageImpact_03}}
 {id=WavePlayer_3, properties {soundWaveAssetPtr=&/BriefcaseTurret_DamageImpact_04}}
 {id=WavePlayer_4, properties {soundWaveAssetPtr=&/BriefcaseTurret_DamageImpact_05}}
 {id=WavePlayer_5, properties {soundWaveAssetPtr=&/BriefcaseTurret_DamageImpact_06}}
]
properties {
 attenuationSettings=/Game/Sounds/Attenuation/Fort_Bullet_Impact_Athena_Att
 maxDistance=4500
 soundClassObject=&Fort_Impacts
 volumeMultiplier=0.64
 concurrencyOverrides {resolutionRule=PreventNew}
}
