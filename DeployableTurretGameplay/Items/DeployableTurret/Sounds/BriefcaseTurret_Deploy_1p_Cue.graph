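# First-person deploy sound cue for the briefcase deployable turret, routed to the gadget items sound class.
# A mixer layers two random nodes, each picking one of three equally weighted 2D variations, with the plain deploy layer mixed below the "B" layer.
# A concurrency override allows one instance per owner and stops the oldest, and overall volume is heavily reduced.
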SoundCue /DeployableTurretGameplay/Items/DeployableTurret/Sounds/BriefcaseTurret_Deploy_1p_Cue
root=Mixer_1
nodes [
 {id=Mixer_1, children [Random_1,Random_2], properties {inputVolume [1,0.7]}}
 {id=Random_1, children [WavePlayer_0,WavePlayer_4,WavePlayer_5], properties {weights [1,1,1]}}
 {id=Random_2, children [WavePlayer_6,WavePlayer_7,WavePlayer_8], properties {weights [1,1,1]}}
 {id=WavePlayer_0, properties {soundWaveAssetPtr=&/BriefcaseTurret_Deploy_2D_B_01}}
 {id=WavePlayer_4, properties {soundWaveAssetPtr=&/BriefcaseTurret_Deploy_2D_B_02}}
 {id=WavePlayer_5, properties {soundWaveAssetPtr=&/BriefcaseTurret_Deploy_2D_B_03}}
 {id=WavePlayer_6, properties {soundWaveAssetPtr=&/BriefcaseTurret_Deploy_2D_01}}
 {id=WavePlayer_7, properties {soundWaveAssetPtr=&/BriefcaseTurret_Deploy_2D_02}}
 {id=WavePlayer_8, properties {soundWaveAssetPtr=&/BriefcaseTurret_Deploy_2D_03}}
]
properties {
 bOverrideConcurrency=true
 maxDistance=2097152
 soundClassObject=&Fort_Gadgets_Items
 volumeMultiplier=0.08
 concurrencyOverrides {bLimitToOwner=true, maxCount=1, resolutionRule=StopOldest}
}
