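# Sound cue for the deployable briefcase turret's deploy sound.
# A mixer layers one of three randomly picked deploy waves with a pair of beep waves that play together after a fixed 1.4-second delay.
# It uses the turret spawn attenuation asset and the gadget item sound class, with a concurrency override that keeps one instance per owner and stops the oldest.
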SoundCue /DeployableTurretGameplay/Items/DeployableTurret/Sounds/BriefcaseTurret_Deploy_Cue
root=Mixer_0
nodes [
 {id=Delay_0, children [Mixer_1], properties {delayMax=1.4, delayMin=1.4}}
 {id=Mixer_0, children [Random_2,Delay_0], properties {inputVolume [1,1]}}
 {id=Mixer_1, children [WavePlayer_11, WavePlayer_14], properties {inputVolume [1,1]}}
 {id=Random_2, children [WavePlayer_6,WavePlayer_7,WavePlayer_8], properties {weights [1,1,1]}}
 {id=WavePlayer_11, properties {soundWaveAssetPtr=&/BriefcaseTurret_Beeps_03}}
 {id=WavePlayer_14, properties {soundWaveAssetPtr=&/BriefcaseTurret_Beeps_06}}
 {id=WavePlayer_6, properties {soundWaveAssetPtr=&/BriefcaseTurret_Deploy_01}}
 {id=WavePlayer_7, properties {soundWaveAssetPtr=&/BriefcaseTurret_Deploy_02}}
 {id=WavePlayer_8, properties {soundWaveAssetPtr=&/BriefcaseTurret_Deploy_03}}
]
properties {
 attenuationSettings=/DeployableTurretGameplay/Items/DeployableTurret/Sounds/Settings/BriefcaseTurret_Spawn_Att
 bOverrideConcurrency=true
 maxDistance=2400
 soundClassObject=&Fort_Gadgets_Items
 volumeMultiplier=0.3
 concurrencyOverrides {bLimitToOwner=true, maxCount=1, resolutionRule=StopOldest}
}
