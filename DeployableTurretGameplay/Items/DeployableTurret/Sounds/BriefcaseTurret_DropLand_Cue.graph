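# Sound cue for the briefcase deployable turret's drop-and-land impact.
# A mixer layers one of three equally weighted random impact variants, passed through a default modulator, over a quieter secondary impact with wider pitch and volume modulation.
# The cue raises overall pitch, lowers volume, and uses shared weapon-drop attenuation and concurrency assets with the loot-collect sound class.
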SoundCue /DeployableTurretGameplay/Items/DeployableTurret/Sounds/BriefcaseTurret_DropLand_Cue
root=Mixer_0
nodes [
 {id=Mixer_0, children [Modulator_0,Modulator_1], properties {inputVolume [1.2,0.8]}}
 {id=Modulator_0, children [Random_0]}
 {id=Modulator_1, children [WavePlayer_6], properties {pitchMax=1.1, pitchMin=0.9, volumeMax=1.1, volumeMin=0.9}}
 {id=Random_0, children [WavePlayer_3,WavePlayer_4,WavePlayer_5], properties {weights [1,1,1]}}
 {id=WavePlayer_3, properties {soundWaveAssetPtr=&/BriefcaseTurret_DropImact_01}}
 {id=WavePlayer_4, properties {soundWaveAssetPtr=&/BriefcaseTurret_DropImact_02}}
 {id=WavePlayer_5, properties {soundWaveAssetPtr=&/BriefcaseTurret_DropImact_03}}
 {id=WavePlayer_6, properties {soundWaveAssetPtr=&/BriefcaseTurret_DropImact_B_01}}
]
properties {
 attenuationSettings=&Fort_Weapon_Drop_Att
 maxDistance=3000
 pitchMultiplier=1.2
 soundClassObject=&Fort_Loot_Collect
 volumeMultiplier=0.6
 concurrencyOverrides {resolutionRule=PreventNew}
 concurrencySet [&Weapon_Drop_StopOldest]
}
