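# Sound cue named as the equip loop for the deployable briefcase turret, with no root or sound nodes, so it has nothing to play.
# It still carries playback settings: the Fort gadget items sound class, a volume boost and the turret's Sight attenuation asset.
# Attenuation override values include listener focus that lowers off-focus volume.
# A concurrency override allows one instance per owner and blocks new ones.
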SoundCue /DeployableTurretGameplay/Items/DeployableTurret/Sounds/BriefcaseTurret_EquipLoop_Cue
root=null
nodes []
properties {
 attenuationSettings=/DeployableTurretGameplay/Items/DeployableTurret/Sounds/Settings/BriefcaseTurret_Sight_Att
 bOverrideConcurrency=true
 maxDistance=4100
 soundClassObject=&Fort_Gadgets_Items
 volumeMultiplier=1.5
 attenuationOverrides {
  bEnableFocusInterpolation=true
  bEnableListenerFocus=true
  falloffDistance=1200
  nonFocusVolumeAttenuation=0.5
  stereoSpread=500
  attenuationShapeExtents {x=0}
 }
 concurrencyOverrides {bLimitToOwner=true, maxCount=1, resolutionRule=PreventNew}
}
