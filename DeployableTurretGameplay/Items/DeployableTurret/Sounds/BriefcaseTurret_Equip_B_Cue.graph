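# Sound cue for equipping the deployable briefcase turret, played as player foley with close prop attenuation.
# A mixer layers two branches: one of three randomly chosen equip variations, and a quieter Equip_B wave.
# Each branch passes through a modulator, with the B layer also pitched down and randomized slightly in pitch and volume.
# A concurrency override prevents new instances once the limit is reached.
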SoundCue /DeployableTurretGameplay/Items/DeployableTurret/Sounds/BriefcaseTurret_Equip_B_Cue
root=Mixer_0
nodes [
 {id=Mixer_0, children [Modulator_0,Modulator_1], properties {inputVolume [1,0.4]}}
 {id=Modulator_0, children [Random_0]}
 {id=Modulator_1, children [WavePlayer_4], properties {pitchMax=0.9, pitchMin=0.8, volumeMax=0.85, volumeMin=0.75}}
 {id=Random_0, children [WavePlayer_1,WavePlayer_2,WavePlayer_3], properties {weights [1,1,1]}}
 {id=WavePlayer_1, properties {soundWaveAssetPtr=&/BriefcaseTurret_Equip_01}}
 {id=WavePlayer_2, properties {soundWaveAssetPtr=&/BriefcaseTurret_Equip_02}}
 {id=WavePlayer_3, properties {soundWaveAssetPtr=&/BriefcaseTurret_Equip_03}}
 {id=WavePlayer_4, properties {soundWaveAssetPtr=&/BriefcaseTurret_Equip_B_01}}
]
properties {
 attenuationSettings=/Game/Sounds/Attenuation/Fort_Prop_Close_Att
 bOverrideConcurrency=true
 maxDistance=1800
 soundClassObject=&Fort_Player_Foley
 volumeMultiplier=1.4
 concurrencyOverrides {resolutionRule=PreventNew}
}
