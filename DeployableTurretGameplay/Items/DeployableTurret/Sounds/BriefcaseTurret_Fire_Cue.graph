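# Sound cue for the briefcase turret's gunfire, layering close, medium and distant shots with per-layer attenuation and sound class.
# A quality switch sends its first two levels to a fuller mix with more shot variants and shell impacts, and the last two to a three-variant mix without shells.
# Random nodes pick waves, and modulators vary playback.
# Concurrency overrides limit voices per owner, stopping the oldest.
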SoundCue /DeployableTurretGameplay/Items/DeployableTurret/Sounds/BriefcaseTurret_Fire_Cue
root=QualityLevel_0
nodes [
 {id=Attenuation_0, children [Modulator_10], properties {attenuationSettings=&Fort_Weapon_Projectile_Close_Athena_Att}}
 {
  id=Attenuation_1
  children [Modulator_9]
  properties {attenuationSettings=&Fort_Weapon_Projectile_Distant_Mono_Athena_Att}
 }
 {
  id=Attenuation_2
  children [Modulator_8]
  properties {attenuationSettings=&Fort_Weapon_Projectile_Medium_Mono_Athena_Att}
 }
 {id=Attenuation_3, children [Modulator_6], properties {attenuationSettings=&Fort_Weapon_Projectile_Close_Athena_Att}}
 {
  id=Attenuation_4
  children [Modulator_5]
  properties {attenuationSettings=&Fort_Weapon_Projectile_Distant_Mono_Athena_Att}
 }
 {
  id=Attenuation_5
  children [Modulator_4]
  properties {attenuationSettings=&Fort_Weapon_Projectile_Medium_Mono_Athena_Att}
 }
 {id=Attenuation_6, children [Modulator_11], properties {attenuationSettings=&Fort_Weapon_Projectile_Close_Athena_Att}}
 {id=Mixer_0, children [Modulator_7,SoundClass_2], properties {inputVolume [1,0.4]}}
 {id=Mixer_2, children [Modulator_3,SoundClass_4], properties {inputVolume [1,0.4]}}
 {id=Mixer_3, children [Mixer_6,SoundClass_5], properties {inputVolume [0.95,0.8]}}
 {id=Mixer_5, children [SoundClass_1,SoundClass_6], properties {inputVolume [0.95,0.8]}}
 {id=Mixer_6, children [SoundClass_3, Attenuation_6], properties {inputVolume [1,0.25]}}
 {id=Modulator_10, children [Random_9]}
 {id=Modulator_11, children [Random_12], properties {pitchMax=1.1, pitchMin=0.9, volumeMax=1.1, volumeMin=0.9}}
 {id=Modulator_3, children [Mixer_3], properties {pitchMax=1, pitchMin=1}}
 {id=Modulator_4, children [Random_7], properties {volumeMax=1, volumeMin=0.9}}
 {id=Modulator_5, children [Random_8]}
 {id=Modulator_6, children [Random_6]}
 {id=Modulator_7, children [Mixer_5], properties {pitchMax=1, pitchMin=1}}
 {id=Modulator_8, children [Random_10], properties {volumeMax=1, volumeMin=0.9}}
 {id=Modulator_9, children [Random_11]}
 {id=QualityLevel_0, children [Mixer_2,Mixer_2,Mixer_0,Mixer_0]}
 {id=Random_10, children [WavePlayer_57, WavePlayer_58, WavePlayer_59], properties {weights [1,1,1]}}
 {id=Random_11, children [WavePlayer_62, WavePlayer_63, WavePlayer_64], properties {weights [1,1,1]}}
 {
  id=Random_12
  children [WavePlayer_67, WavePlayer_68, WavePlayer_69, WavePlayer_70, WavePlayer_71, WavePlayer_72]
  properties {weights [1,1,1,1,1,1]}
 }
 {
  id=Random_6
  children [WavePlayer_31, WavePlayer_32, WavePlayer_33, WavePlayer_34, WavePlayer_35, WavePlayer_36, WavePlayer_37, WavePlayer_38]
  properties {weights [1,1,1,1,1,1,1,1]}
 }
 {id=Random_7, children [WavePlayer_39, WavePlayer_40, WavePlayer_41, WavePlayer_42, WavePlayer_43], properties {weights [1,1,1,1,1]}}
 {id=Random_8, children [WavePlayer_44, WavePlayer_45, WavePlayer_46, WavePlayer_47, WavePlayer_48], properties {weights [1,1,1,1,1]}}
 {id=Random_9, children [WavePlayer_49, WavePlayer_50, WavePlayer_51], properties {weights [1,1,1]}}
 {
  id=SoundClass_1
  children [Attenuation_0]
  properties {soundClassOverride=&Fort_WeaponRanged_Fire_3P_PSM}
 }
 {
  id=SoundClass_2
  children [Attenuation_1]
  properties {soundClassOverride=&Fort_WeaponRanged_Fire_Distant}
 }
 {
  id=SoundClass_3
  children [Attenuation_3]
  properties {soundClassOverride=&Fort_WeaponRanged_Fire_3P_PSM}
 }
 {
  id=SoundClass_4
  children [Attenuation_4]
  properties {soundClassOverride=&Fort_WeaponRanged_Fire_Distant}
 }
 {
  id=SoundClass_5
  children [Attenuation_5]
  properties {soundClassOverride=&Fort_WeaponRanged_Fire_Distant}
 }
 {
  id=SoundClass_6
  children [Attenuation_2]
  properties {soundClassOverride=&Fort_WeaponRanged_Fire_Distant}
 }
 {id=WavePlayer_31, properties {soundWaveAssetPtr=&/BriefcaseTurret_Fire_01}}
 {id=WavePlayer_32, properties {soundWaveAssetPtr=&/BriefcaseTurret_Fire_02}}
 {id=WavePlayer_33, properties {soundWaveAssetPtr=&/BriefcaseTurret_Fire_03}}
 {id=WavePlayer_34, properties {soundWaveAssetPtr=&/BriefcaseTurret_Fire_04}}
 {id=WavePlayer_35, properties {soundWaveAssetPtr=&/BriefcaseTurret_Fire_05}}
 {id=WavePlayer_36, properties {soundWaveAssetPtr=&/BriefcaseTurret_Fire_06}}
 {id=WavePlayer_37, properties {soundWaveAssetPtr=&/BriefcaseTurret_Fire_07}}
 {id=WavePlayer_38, properties {soundWaveAssetPtr=&/BriefcaseTurret_Fire_08}}
 {id=WavePlayer_39, properties {soundWaveAssetPtr=&/BriefcaseTurret_Fire_Medium_01}}
 {id=WavePlayer_40, properties {soundWaveAssetPtr=&/BriefcaseTurret_Fire_Medium_02}}
 {id=WavePlayer_41, properties {soundWaveAssetPtr=&/BriefcaseTurret_Fire_Medium_03}}
 {id=WavePlayer_42, properties {soundWaveAssetPtr=&/BriefcaseTurret_Fire_Medium_04}}
 {id=WavePlayer_43, properties {soundWaveAssetPtr=&/BriefcaseTurret_Fire_Medium_05}}
 {id=WavePlayer_44, properties {soundWaveAssetPtr=&/BriefcaseTurret_Fire_Distant_01}}
 {id=WavePlayer_45, properties {soundWaveAssetPtr=&/BriefcaseTurret_Fire_Distant_02}}
 {id=WavePlayer_46, properties {soundWaveAssetPtr=&/BriefcaseTurret_Fire_Distant_03}}
 {id=WavePlayer_47, properties {soundWaveAssetPtr=&/BriefcaseTurret_Fire_Distant_04}}
 {id=WavePlayer_48, properties {soundWaveAssetPtr=&/BriefcaseTurret_Fire_Distant_05}}
 {id=WavePlayer_49, properties {soundWaveAssetPtr=&BriefcaseTurret_Fire_01}}
 {id=WavePlayer_50, properties {soundWaveAssetPtr=&BriefcaseTurret_Fire_02}}
 {id=WavePlayer_51, properties {soundWaveAssetPtr=&BriefcaseTurret_Fire_03}}
 {id=WavePlayer_57, properties {soundWaveAssetPtr=&BriefcaseTurret_Fire_Medium_01}}
 {id=WavePlayer_58, properties {soundWaveAssetPtr=&BriefcaseTurret_Fire_Medium_02}}
 {id=WavePlayer_59, properties {soundWaveAssetPtr=&BriefcaseTurret_Fire_Medium_03}}
 {id=WavePlayer_62, properties {soundWaveAssetPtr=&BriefcaseTurret_Fire_Distant_01}}
 {id=WavePlayer_63, properties {soundWaveAssetPtr=&BriefcaseTurret_Fire_Distant_02}}
 {id=WavePlayer_64, properties {soundWaveAssetPtr=&BriefcaseTurret_Fire_Distant_03}}
 {id=WavePlayer_67, properties {soundWaveAssetPtr=&/BriefcaseTurret_ShellImpact_01}}
 {id=WavePlayer_68, properties {soundWaveAssetPtr=&/BriefcaseTurret_ShellImpact_02}}
 {id=WavePlayer_69, properties {soundWaveAssetPtr=&/BriefcaseTurret_ShellImpact_03}}
 {id=WavePlayer_70, properties {soundWaveAssetPtr=&/BriefcaseTurret_ShellImpact_04}}
 {id=WavePlayer_71, properties {soundWaveAssetPtr=&/BriefcaseTurret_ShellImpact_05}}
 {id=WavePlayer_72, properties {soundWaveAssetPtr=&/BriefcaseTurret_ShellImpact_06}}
]
properties {
 bDebug=true
 maxDistance=27500
 soundClassObject=&Fort_WeaponRanged_Fire_3P_PSM
 volumeMultiplier=0.4
 concurrencyOverrides {bLimitToOwner=true, maxCount=3, resolutionRule=StopOldest}
 concurrencySet [&Weapon_Gunfire_Local_Conc, &Weapon_Gunfire_Global_Conc]
}
