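# Sound cue for the deployable briefcase turret's fire lift-down sound, in the gadgets and items sound class.
# An equally weighted random node picks one of three wave variations and feeds it through a modulator at the root.
# The cue plays at reduced volume with the turret's ambient attenuation settings.
# A concurrency override allows only one instance at a time, stopping the oldest when a new one starts.
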SoundCue /DeployableTurretGameplay/Items/DeployableTurret/Sounds/BriefcaseTurret_Fire_LiftDown_Cue
root=Modulator_1
nodes [
 {id=Modulator_1, children [Random_1]}
 {id=Random_1, children [WavePlayer_0, WavePlayer_10, WavePlayer_11], properties {weights [1,1,1]}}
 {id=WavePlayer_0, properties {soundWaveAssetPtr=&/BriefcaseTurret_Fire_LiftDown_01}}
 {id=WavePlayer_10, properties {soundWaveAssetPtr=&/BriefcaseTurret_Fire_LiftDown_02}}
 {id=WavePlayer_11, properties {soundWaveAssetPtr=&/BriefcaseTurret_Fire_LiftDown_03}}
]
properties {
 attenuationSettings=/DeployableTurretGameplay/Items/DeployableTurret/Sounds/Settings/BriefcaseTurret_Ambient_Att
 bOverrideConcurrency=true
 maxDistance=1050
 soundClassObject=&Fort_Gadgets_Items
 volumeMultiplier=0.6
 concurrencyOverrides {maxCount=1, resolutionRule=StopOldest}
}
