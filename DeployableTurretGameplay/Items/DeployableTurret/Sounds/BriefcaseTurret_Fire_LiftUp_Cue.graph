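# Sound cue for the briefcase deployable turret's fire lift-up sound, routed to the gadget items sound class.
# A random node picks one of three equally weighted wave variations and feeds it through a modulator.
# It references a turret sight attenuation asset and reduces overall volume.
# A concurrency override allows one instance per owner and blocks new plays while one is active.
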SoundCue /DeployableTurretGameplay/Items/DeployableTurret/Sounds/BriefcaseTurret_Fire_LiftUp_Cue
root=Modulator_0
nodes [
 {id=Modulator_0, children [Random_0]}
 {id=Random_0, children [WavePlayer_1,WavePlayer_2,WavePlayer_3], properties {weights [1,1,1]}}
 {id=WavePlayer_1, properties {soundWaveAssetPtr=&/BriefcaseTurret_Fire_LiftUp_01}}
 {id=WavePlayer_2, properties {soundWaveAssetPtr=&/BriefcaseTurret_Fire_LiftUp_02}}
 {id=WavePlayer_3, properties {soundWaveAssetPtr=&/BriefcaseTurret_Fire_LiftUp_03}}
]
properties {
 attenuationSettings=/DeployableTurretGameplay/Items/DeployableTurret/Sounds/Settings/BriefcaseTurret_Sight_Att
 bOverrideConcurrency=true
 maxDistance=4100
 soundClassObject=&Fort_Gadgets_Items
 volumeMultiplier=0.7
 attenuationOverrides {
  bEnableFocusInterpolation=true
  bEnableListenerFocus=true
  falloffDistance=1200
  nonFocusVolumeAttenuation=0.5
  stereoSpread=500
  attenuationShapeExtents {x=0}
 }
 concurrencyOverrides {bLimitToOwner=true, maxCount=1, resolutionRule=PreventNew}
}
