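# Sound cue for the deployable briefcase turret's fire power-up sound, routed to the gadget items sound class.
# A random node picks between two equally weighted power-up waves and feeds a modulator that plays them at a fixed, slightly lowered pitch.
# It references the turret sight attenuation asset, with listener-focus overrides also defined.
# Concurrency allows one instance per owner and blocks new plays while one is active.
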SoundCue /DeployableTurretGameplay/Items/DeployableTurret/Sounds/BriefcaseTurret_Fire_PowerUp_Cue
root=Modulator_0
nodes [
 {id=Modulator_0, children [Random_0], properties {pitchMax=0.9, pitchMin=0.9}}
 {id=Random_0, children [WavePlayer_1,WavePlayer_2], properties {weights [1,1]}}
 {id=WavePlayer_1, properties {soundWaveAssetPtr=&/BriefcaseTurret_Fire_Powerup_01}}
 {id=WavePlayer_2, properties {soundWaveAssetPtr=&/BriefcaseTurret_Fire_Powerup_02}}
]
properties {
 attenuationSettings=/DeployableTurretGameplay/Items/DeployableTurret/Sounds/Settings/BriefcaseTurret_Sight_Att
 bOverrideConcurrency=true
 maxDistance=4100
 soundClassObject=&Fort_Gadgets_Items
 volumeMultiplier=0.8
 attenuationOverrides {
  bEnableFocusInterpolation=true
  bEnableListenerFocus=true
  falloffDistance=1200
  nonFocusVolumeAttenuation=0.5
  stereoSpread=500
  attenuationShapeExtents {x=0}
 }
 concurrencyOverrides {bLimitToOwner=true, maxCount=1, resolutionRule=PreventNew}
}
