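# Sound cue for the looping in-air sound of the deployable briefcase turret, played at half volume in the gadget items sound class.
# A Random node picks one of two equally weighted looping waves and feeds it through a Modulator to the output.
# Concurrency overrides allow one instance per owner and block new ones while it plays.
# Spatial falloff references the turret's Sight attenuation asset rather than one specific to this loop.
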SoundCue /DeployableTurretGameplay/Items/DeployableTurret/Sounds/BriefcaseTurret_InAirLoop_Cue
root=Modulator_0
nodes [
 {id=Modulator_0, children [Random_0]}
 {id=Random_0, children [WavePlayer_0,WavePlayer_2], properties {weights [1,1]}}
 {
  id=WavePlayer_0
  properties {bLooping=true, soundWaveAssetPtr=&/BriefcaseTurret_InAirLoop_01}
 }
 {
  id=WavePlayer_2
  properties {bLooping=true, soundWaveAssetPtr=&/BriefcaseTurret_InAirLoop_02}
 }
]
properties {
 attenuationSettings=/DeployableTurretGameplay/Items/DeployableTurret/Sounds/Settings/BriefcaseTurret_Sight_Att
 bOverrideConcurrency=true
 maxDistance=4100
 soundClassObject=&Fort_Gadgets_Items
 volumeMultiplier=0.5
 attenuationOverrides {
  bEnableFocusInterpolation=true
  bEnableListenerFocus=true
  falloffDistance=1200
  nonFocusVolumeAttenuation=0.5
  stereoSpread=500
  attenuationShapeExtents {x=0}
 }
 concurrencyOverrides {bLimitToOwner=true, maxCount=1, resolutionRule=PreventNew}
}
