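# Sound cue for the deployable briefcase turret landing and unfolding, layering the impact with its mechanical follow-up sounds.
# A mixer plays one of three random impacts immediately, then staggers a random latch, a spin and a later extend through fixed delays, each pitched up.
# It routes to the gadget items sound class, uses the turret's Sight attenuation asset and allows one instance per owner, refusing new ones.
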SoundCue /DeployableTurretGameplay/Items/DeployableTurret/Sounds/BriefcaseTurret_LandImpact_Cue
root=Mixer_0
nodes [
 {id=Delay_0, children [Modulator_1], properties {delayMax=0.4, delayMin=0.4}}
 {id=Delay_1, children [Modulator_2], properties {delayMax=1.1, delayMin=1.1}}
 {id=Delay_2, children [Modulator_3], properties {delayMax=0.3, delayMin=0.3}}
 {id=Mixer_0, children [Modulator_0,Delay_0,Delay_1,Delay_2], properties {inputVolume [0.8,1.2,1,1]}}
 {id=Modulator_0, children [Random_0]}
 {id=Modulator_1, children [WavePlayer_1], properties {pitchMax=1.1, pitchMin=1.1}}
 {id=Modulator_2, children [WavePlayer_4], properties {pitchMax=1.1, pitchMin=1.1}}
 {id=Modulator_3, children [Random_1], properties {pitchMax=1.2, pitchMin=1.2}}
 {id=Random_0, children [WavePlayer_0,WavePlayer_2,WavePlayer_3], properties {weights [1,1,1]}}
 {id=Random_1, children [WavePlayer_5,WavePlayer_6], properties {weights [1,1]}}
 {id=WavePlayer_0, properties {soundWaveAssetPtr=&/BriefcaseTurret_LandImpact_01}}
 {id=WavePlayer_1, properties {soundWaveAssetPtr=&/BriefcaseTurret_Land_Spin_01}}
 {id=WavePlayer_2, properties {soundWaveAssetPtr=&/BriefcaseTurret_LandImpact_02}}
 {id=WavePlayer_3, properties {soundWaveAssetPtr=&/BriefcaseTurret_LandImpact_03}}
 {id=WavePlayer_4, properties {soundWaveAssetPtr=&/BriefcaseTurret_Land_Extend_01}}
 {id=WavePlayer_5, properties {soundWaveAssetPtr=&/BriefcaseTurret_Land_Latch_01}}
 {id=WavePlayer_6, properties {soundWaveAssetPtr=&/BriefcaseTurret_Land_Latch_02}}
]
properties {
 attenuationSettings=/DeployableTurretGameplay/Items/DeployableTurret/Sounds/Settings/BriefcaseTurret_Sight_Att
 bOverrideConcurrency=true
 maxDistance=4100
 soundClassObject=&Fort_Gadgets_Items
 volumeMultiplier=0.8
 attenuationOverrides {
  bEnableFocusInterpolation=true
  bEnableListenerFocus=true
  falloffDistance=1200
  nonFocusVolumeAttenuation=0.5
  stereoSpread=500
  attenuationShapeExtents {x=0}
 }
 concurrencyOverrides {bLimitToOwner=true, maxCount=1, resolutionRule=PreventNew}
}
