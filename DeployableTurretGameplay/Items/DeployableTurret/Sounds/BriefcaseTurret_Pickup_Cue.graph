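# Sound cue for picking up the briefcase deployable turret.
# It randomly selects one of three equally weighted briefcase pickup waves, then passes it through a Modulator at default settings.
# Playback uses the loot collect sound class and weapon pickup attenuation, with a shared pickup concurrency set limiting overlapping sounds.
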SoundCue /DeployableTurretGameplay/Items/DeployableTurret/Sounds/BriefcaseTurret_Pickup_Cue
root=Modulator_0
nodes [
 {id=Modulator_0, children [Random_1]}
 {id=Random_1, children [WavePlayer_1,WavePlayer_3,WavePlayer_4], properties {weights [1,1,1]}}
 {id=WavePlayer_1, properties {soundWaveAssetPtr=&/BriefcaseTurret_Pickup_Briefcase_01}}
 {id=WavePlayer_3, properties {soundWaveAssetPtr=&/BriefcaseTurret_Pickup_Briefcase_02}}
 {id=WavePlayer_4, properties {soundWaveAssetPtr=&/BriefcaseTurret_Pickup_Briefcase_03}}
]
properties {
 attenuationSettings=&Fort_Weapon_Pickup_Att
 maxDistance=3500
 soundClassObject=&Fort_Loot_Collect
 volumeMultiplier=1
 concurrencyOverrides {maxCount=1, resolutionRule=PreventNew}
 concurrencySet [&PickupMaximum_Oldest]
}
