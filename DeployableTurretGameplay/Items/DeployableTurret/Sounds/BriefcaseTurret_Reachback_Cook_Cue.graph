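# Sound cue for the briefcase deployable turret's "reachback cook" sound, classed as a gadget sound and using weapon reload attenuation.
# It picks one of three equally weighted reachback waves, passes it through a modulator at default settings, and starts it after a short fixed delay.
# The cue as a whole plays at reduced volume and slightly lowered pitch.
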SoundCue /DeployableTurretGameplay/Items/DeployableTurret/Sounds/BriefcaseTurret_Reachback_Cook_Cue
root=Delay_0
nodes [
 {id=Delay_0, children [Modulator_0], properties {delayMax=0.05, delayMin=0.05}}
 {id=Modulator_0, children [Random_1]}
 {id=Random_1, children [WavePlayer_3,WavePlayer_4,WavePlayer_5], properties {weights [1,1,1]}}
 {id=WavePlayer_3, properties {soundWaveAssetPtr=&/BriefcaseTurret_Reachback_01}}
 {id=WavePlayer_4, properties {soundWaveAssetPtr=&/BriefcaseTurret_Reachback_02}}
 {id=WavePlayer_5, properties {soundWaveAssetPtr=&/BriefcaseTurret_Reachback_03}}
]
properties {
 attenuationSettings=&Fort_Weapon_Reload_Att
 maxDistance=2250
 pitchMultiplier=0.9
 soundClassObject=&Fort_Gadgets
 volumeMultiplier=0.3
}
