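# Sound cue for the looping rotation sound of the deployable briefcase turret.
# One of three looping rotate variants is picked at random and mixed with a quieter constant B-layer loop, then boosted by an outer modulator.
# It uses the turret movement attenuation asset, the gadget items sound class, and a concurrency override that caps instances by stopping the quietest.
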SoundCue /DeployableTurretGameplay/Items/DeployableTurret/Sounds/BriefcaseTurret_Rotate_Loop_Cue
root=Modulator_5
nodes [
 {id=Mixer_1, children [Modulator_4,WavePlayer_9], properties {inputVolume [1,0.2]}}
 {id=Modulator_3, children [WavePlayer_1], properties {volumeMax=1.2, volumeMin=1.2}}
 {id=Modulator_4, children [Random_2]}
 {id=Modulator_5, children [Mixer_1], properties {volumeMax=1.5, volumeMin=1.5}}
 {id=Random_2, children [Modulator_3,WavePlayer_7,WavePlayer_8], properties {weights [1,1,1]}}
 {
  id=WavePlayer_1
  properties {bLooping=true, soundWaveAssetPtr=&/BriefcaseTurret_Rotate_Loop_03}
 }
 {
  id=WavePlayer_7
  properties {bLooping=true, soundWaveAssetPtr=&/BriefcaseTurret_Rotate_Loop_04}
 }
 {
  id=WavePlayer_8
  properties {bLooping=true, soundWaveAssetPtr=&/BriefcaseTurret_Rotate_Loop_05}
 }
 {
  id=WavePlayer_9
  properties {bLooping=true, soundWaveAssetPtr=&/BriefcaseTurret_Rotate_Loop_B_01}
 }
]
properties {
 attenuationSettings=/DeployableTurretGameplay/Items/DeployableTurret/Sounds/Settings/BriefcaseTurret_Movement_Att
 bOverrideConcurrency=true
 maxDistance=1600
 soundClassObject=&Fort_Gadgets_Items
 volumeMultiplier=0.5
 attenuationOverrides {
  bEnableFocusInterpolation=true
  bEnableListenerFocus=true
  falloffDistance=1200
  nonFocusVolumeAttenuation=0.5
  stereoSpread=500
  attenuationShapeExtents {x=0}
 }
 concurrencyOverrides {maxCount=3, resolutionRule=StopQuietest}
}
