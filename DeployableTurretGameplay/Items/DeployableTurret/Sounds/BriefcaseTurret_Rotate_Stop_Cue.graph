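# Sound cue for the deployable briefcase turret's rotate-stop sound, classed under gadget items.
# A Random node picks among six equally weighted stop wave variations, which feed a Modulator that varies pitch slightly below normal.
# It uses the turret movement attenuation asset with focus-related overrides, and overrides concurrency to a few voices, stopping the quietest.
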SoundCue /DeployableTurretGameplay/Items/DeployableTurret/Sounds/BriefcaseTurret_Rotate_Stop_Cue
root=Modulator_0
nodes [
 {id=Modulator_0, children [Random_1], properties {pitchMax=0.95, pitchMin=0.85}}
 {id=Random_1, children [WavePlayer_1,WavePlayer_4,WavePlayer_5,WavePlayer_6,WavePlayer_7,WavePlayer_8], properties {weights [1,1,1,1,1,1]}}
 {id=WavePlayer_1, properties {soundWaveAssetPtr=&/BriefcaseTurret_Rotate_Stop_01}}
 {id=WavePlayer_4, properties {soundWaveAssetPtr=&/BriefcaseTurret_Rotate_Stop_02}}
 {id=WavePlayer_5, properties {soundWaveAssetPtr=&/BriefcaseTurret_Rotate_Stop_03}}
 {id=WavePlayer_6, properties {soundWaveAssetPtr=&/BriefcaseTurret_Rotate_Stop_04}}
 {id=WavePlayer_7, properties {soundWaveAssetPtr=&/BriefcaseTurret_Rotate_Stop_05}}
 {id=WavePlayer_8, properties {soundWaveAssetPtr=&/BriefcaseTurret_Rotate_Stop_06}}
]
properties {
 attenuationSettings=/DeployableTurretGameplay/Items/DeployableTurret/Sounds/Settings/BriefcaseTurret_Movement_Att
 bOverrideConcurrency=true
 maxDistance=1600
 soundClassObject=&Fort_Gadgets_Items
 volumeMultiplier=1.4
 attenuationOverrides {
  bEnableFocusInterpolation=true
  bEnableListenerFocus=true
  falloffDistance=1200
  nonFocusVolumeAttenuation=0.5
  stereoSpread=500
  attenuationShapeExtents {x=0}
 }
 concurrencyOverrides {maxCount=3, resolutionRule=StopQuietest}
}
