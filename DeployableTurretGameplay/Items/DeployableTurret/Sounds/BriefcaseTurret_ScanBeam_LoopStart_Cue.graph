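# Sound cue for the scan beam start on the briefcase deployable turret, layering beeps with two beam start sounds.
# A mixer plays a random beep immediately, then a modulated random start variant and a boosted random "B" start variant, each after its own fixed delay.
# Overrides add HRTF spatialization, quieter non-focused playback and a small concurrency limit that stops the quietest instance.
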SoundCue /DeployableTurretGameplay/Items/DeployableTurret/Sounds/BriefcaseTurret_ScanBeam_LoopStart_Cue
root=Mixer_0
nodes [
 {id=Delay_0, children [Modulator_0], properties {delayMax=0.4, delayMin=0.4}}
 {id=Delay_1, children [Random_1], properties {delayMax=0.7, delayMin=0.7}}
 {id=Mixer_0, children [Delay_0,Delay_1,Random_2], properties {inputVolume [1,1.5,1]}}
 {id=Modulator_0, children [Random_0]}
 {id=Random_0, children [WavePlayer_0,WavePlayer_1,WavePlayer_2], properties {weights [1,1,1]}}
 {id=Random_1, children [WavePlayer_3,WavePlayer_4,WavePlayer_5], properties {weights [1,1,1]}}
 {id=Random_2, children [WavePlayer_6,WavePlayer_8], properties {weights [1,1]}}
 {
  id=WavePlayer_0
  properties {soundWaveAssetPtr=&/BriefcaseTurret_ScanningBeam_Start_01}
 }
 {
  id=WavePlayer_1
  properties {soundWaveAssetPtr=&/BriefcaseTurret_ScanningBeam_Start_02}
 }
 {
  id=WavePlayer_2
  properties {soundWaveAssetPtr=&/BriefcaseTurret_ScanningBeam_Start_03}
 }
 {
  id=WavePlayer_3
  properties {soundWaveAssetPtr=&/BriefcaseTurret_ScanningBeam_Start_B_01}
 }
 {
  id=WavePlayer_4
  properties {soundWaveAssetPtr=&/BriefcaseTurret_ScanningBeam_Start_B_02}
 }
 {
  id=WavePlayer_5
  properties {soundWaveAssetPtr=&/BriefcaseTurret_ScanningBeam_Start_B_03}
 }
 {id=WavePlayer_6, properties {soundWaveAssetPtr=&/BriefcaseTurret_Beeps_06}}
 {id=WavePlayer_8, properties {soundWaveAssetPtr=&/BriefcaseTurret_Beeps_05}}
]
properties {
 bOverrideAttenuation=true
 bOverrideConcurrency=true
 maxDistance=2000
 soundClassObject=&Fort_Gadgets_Items
 volumeMultiplier=0.6
 attenuationOverrides {
  bEnableFocusInterpolation=true
  falloffDistance=2000
  nonFocusVolumeAttenuation=0.5
  spatializationAlgorithm=SPATIALIZATION_HRTF
  stereoSpread=500
  attenuationShapeExtents {x=0}
 }
 concurrencyOverrides {maxCount=3, resolutionRule=StopQuietest}
}
