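# Looping sound cue for the deployable briefcase turret's scanning beam, routed to the gadget items sound class.
# A mixer layers two looping waves, each with its own attenuation: the main loop uses the Beam settings asset through a single-entry random node.
# The quieter B loop uses overridden wide-beam attenuation and passes through a modulator that raises its pitch slightly.
# A concurrency override caps simultaneous instances, stopping the quietest.
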SoundCue /DeployableTurretGameplay/Items/DeployableTurret/Sounds/BriefcaseTurret_ScanBeam_Loop_Cue
root=Mixer_0
nodes [
 {
  id=Attenuation_0
  children [Random_0]
  properties {
   attenuationSettings=/DeployableTurretGameplay/Items/DeployableTurret/Sounds/Settings/BriefcaseTurret_Beam_Att
   attenuationOverrides {attenuationShape=Cone, falloffDistance=3000, attenuationShapeExtents {y=45, z=45}}
  }
 }
 {
  id=Attenuation_1
  children [WavePlayer_7]
  properties {
   attenuationSettings=/DeployableTurretGameplay/Items/DeployableTurret/Sounds/Settings/BriefcaseTurret_Beam_Wide_Att
   bOverrideAttenuation=true
   attenuationOverrides {falloffDistance=2200, attenuationShapeExtents {x=200}}
  }
 }
 {id=Mixer_0, children [Attenuation_0, Modulator_0], properties {inputVolume [0.8,0.45]}}
 {id=Modulator_0, children [Attenuation_1], properties {pitchMax=1.3, pitchMin=1.2}}
 {id=Random_0, children [WavePlayer_3], properties {weights [1]}}
 {
  id=WavePlayer_3
  properties {bLooping=true, soundWaveAssetPtr=&/BriefcaseTurret_ScanningBeam_Loop_01}
 }
 {
  id=WavePlayer_7
  properties {bLooping=true, soundWaveAssetPtr=&/BriefcaseTurret_ScanningBeam_Loop_B_01}
 }
]
properties {
 bOverrideConcurrency=true
 maxDistance=3200
 soundClassObject=&Fort_Gadgets_Items
 volumeMultiplier=1
 attenuationOverrides {
  bEnableFocusInterpolation=true
  bEnableListenerFocus=true
  falloffDistance=1200
  nonFocusVolumeAttenuation=0.5
  stereoSpread=500
  attenuationShapeExtents {x=0}
 }
 concurrencyOverrides {maxCount=3, resolutionRule=StopQuietest}
}
