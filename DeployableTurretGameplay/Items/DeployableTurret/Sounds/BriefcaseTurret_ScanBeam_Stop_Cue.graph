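# Sound cue for the deployable briefcase turret's scanning beam stop sound.
# A root Random node picks one of three equally weighted stop wave variations per play.
# It uses a dedicated beam-stop attenuation asset, the gadgets and items sound class and a reduced volume.
# Overridden concurrency allows one instance per owner, stopping the oldest when a new one starts.
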SoundCue /DeployableTurretGameplay/Items/DeployableTurret/Sounds/BriefcaseTurret_ScanBeam_Stop_Cue
root=Random_1
nodes [
 {id=Random_1, children [WavePlayer_0,WavePlayer_4,WavePlayer_5], properties {weights [1,1,1]}}
 {id=WavePlayer_0, properties {soundWaveAssetPtr=&/BriefcaseTurret_ScanningBeam_Stop_01}}
 {id=WavePlayer_4, properties {soundWaveAssetPtr=&/BriefcaseTurret_ScanningBeam_Stop_02}}
 {id=WavePlayer_5, properties {soundWaveAssetPtr=&/BriefcaseTurret_ScanningBeam_Stop_03}}
]
properties {
 attenuationSettings=/DeployableTurretGameplay/Items/DeployableTurret/Sounds/Settings/BriefcaseTurret_Beam_Stop_Att
 bOverrideConcurrency=true
 maxDistance=2200
 soundClassObject=&Fort_Gadgets_Items
 volumeMultiplier=0.8
 concurrencyOverrides {bLimitToOwner=true, maxCount=1, resolutionRule=StopOldest}
}
