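# Looping sound cue for the deployable briefcase turret's targeting beam, played in the gadget sound class.
# A mixer layers a random one-shot start sound at half volume with a random looping beam sound.
# That mix feeds two attenuation nodes, close and far, which a distance crossfade blends as the listener moves away, then a root modulator.
# Concurrency is overridden to a few instances, stopping the quietest.
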SoundCue /DeployableTurretGameplay/Items/DeployableTurret/Sounds/BriefcaseTurret_TargetBeam_Loop_Cue
root=Modulator_0
nodes [
 {
  id=Attenuation_0
  children [Mixer_0]
  properties {
   attenuationSettings=/DeployableTurretGameplay/Items/DeployableTurret/Sounds/Settings/BriefcaseTurret_Beam_Close_Att
   attenuationOverrides {falloffDistance=500, attenuationShapeExtents {x=0}}
  }
 }
 {
  id=Attenuation_1
  children [Mixer_0]
  properties {attenuationSettings=/DeployableTurretGameplay/Items/DeployableTurret/Sounds/Settings/BriefcaseTurret_Beam_Att}
 }
 {
  id=DistanceCrossFade_1
  children [Attenuation_1, Attenuation_0]
  properties {crossFadeInput [
    {fadeInDistanceEnd=700, fadeInDistanceStart=400, fadeOutDistanceEnd=3000, fadeOutDistanceStart=2500, volume=1}
    {fadeInDistanceEnd=0, fadeInDistanceStart=0, fadeOutDistanceEnd=700, fadeOutDistanceStart=400, volume=0.8}
   ]}
 }
 {id=Mixer_0, children [Random_1,Random_0], properties {inputVolume [0.5,1]}}
 {id=Modulator_0, children [DistanceCrossFade_1]}
 {id=Random_0, children [WavePlayer_3,WavePlayer_5], properties {weights [1,1]}}
 {id=Random_1, children [WavePlayer_6,WavePlayer_7,WavePlayer_8], properties {weights [1,1,1]}}
 {
  id=WavePlayer_3
  properties {bLooping=true, soundWaveAssetPtr=&/BriefcaseTurret_TargetingBeam_Loop_01}
 }
 {
  id=WavePlayer_5
  properties {bLooping=true, soundWaveAssetPtr=&/BriefcaseTurret_TargetingBeam_Loop_02}
 }
 {
  id=WavePlayer_6
  properties {soundWaveAssetPtr=&/BriefcaseTurret_TargetingBeam_Start_01}
 }
 {
  id=WavePlayer_7
  properties {soundWaveAssetPtr=&/BriefcaseTurret_TargetingBeam_Start_02}
 }
 {
  id=WavePlayer_8
  properties {soundWaveAssetPtr=&/BriefcaseTurret_TargetingBeam_Start_03}
 }
]
properties {
 bOverrideConcurrency=true
 maxDistance=3200
 soundClassObject=&Fort_Gadgets_Items
 volumeMultiplier=0.8
 attenuationOverrides {
  bEnableFocusInterpolation=true
  bEnableListenerFocus=true
  falloffDistance=1200
  nonFocusVolumeAttenuation=0.5
  stereoSpread=500
  attenuationShapeExtents {x=0}
 }
 concurrencyOverrides {maxCount=3, resolutionRule=StopQuietest}
}
