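# Sound cue for the deployable briefcase turret's targeting beam alarm that plays when the turret loses its target.
# A random node picks one of three equally weighted target-lost waves and feeds it through a modulator for variation.
# It uses the turret sight attenuation and the gadget items sound class.
# A concurrency override keeps a single instance per owner, stopping the oldest when a new one starts.
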SoundCue /DeployableTurretGameplay/Items/DeployableTurret/Sounds/BriefcaseTurret_TargetBeam_Lost_Cue
root=Modulator_1
nodes [
 {id=Modulator_1, children [Random_0]}
 {id=Random_0, children [WavePlayer_0,WavePlayer_1,WavePlayer_3], properties {weights [1,1,1]}}
 {
  id=WavePlayer_0
  properties {soundWaveAssetPtr=&/BriefcaseTurret_TargetingBeam_Alarm_TargetLost_01}
 }
 {
  id=WavePlayer_1
  properties {soundWaveAssetPtr=&/BriefcaseTurret_TargetingBeam_Alarm_TargetLost_02}
 }
 {
  id=WavePlayer_3
  properties {soundWaveAssetPtr=&/BriefcaseTurret_TargetingBeam_Alarm_TargetLost_03}
 }
]
properties {
 attenuationSettings=/DeployableTurretGameplay/Items/DeployableTurret/Sounds/Settings/BriefcaseTurret_Sight_Att
 bOverrideConcurrency=true
 maxDistance=4100
 soundClassObject=&Fort_Gadgets_Items
 volumeMultiplier=0.8
 concurrencyOverrides {bLimitToOwner=true, maxCount=1, resolutionRule=StopOldest}
}
