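# Sound cue for the deployable briefcase turret's wind-down sound, named for when its targeting beam loses a target.
# It mixes a randomly chosen wind-down variant, pitched down by a fixed modulator, with a quieter, slightly delayed secondary layer under a root modulator.
# Attenuation uses the turret's sight settings, and a concurrency override allows one instance per owner, stopping the oldest.
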SoundCue /DeployableTurretGameplay/Items/DeployableTurret/Sounds/BriefcaseTurret_TargetBeam_Lost_WindDown_Cue
root=Modulator_2
nodes [
 {id=Delay_0, children [WavePlayer_0], properties {delayMax=0.15, delayMin=0.15}}
 {id=Mixer_0, children [Delay_0,Modulator_1], properties {inputVolume [0.4,1]}}
 {id=Modulator_1, children [Random_0], properties {pitchMax=0.8, pitchMin=0.8}}
 {id=Modulator_2, children [Mixer_0]}
 {id=Random_0, children [WavePlayer_1,WavePlayer_5,WavePlayer_6], properties {weights [1,1,1]}}
 {
  id=WavePlayer_0
  properties {soundWaveAssetPtr=&/BriefcaseTurret_TargetBeam_Lost_WindDown_B}
 }
 {
  id=WavePlayer_1
  properties {soundWaveAssetPtr=&/BriefcaseTurret_TargetBeam_Lost_WindDown_01}
 }
 {
  id=WavePlayer_5
  properties {soundWaveAssetPtr=&/BriefcaseTurret_TargetBeam_Lost_WindDown_02}
 }
 {
  id=WavePlayer_6
  properties {soundWaveAssetPtr=&/BriefcaseTurret_TargetBeam_Lost_WindDown_03}
 }
]
properties {
 attenuationSettings=/DeployableTurretGameplay/Items/DeployableTurret/Sounds/Settings/BriefcaseTurret_Sight_Att
 bOverrideConcurrency=true
 maxDistance=4100
 pitchMultiplier=0.9
 soundClassObject=&Fort_Gadgets_Items
 volumeMultiplier=0.8
 concurrencyOverrides {bLimitToOwner=true, maxCount=1, resolutionRule=StopOldest}
}
