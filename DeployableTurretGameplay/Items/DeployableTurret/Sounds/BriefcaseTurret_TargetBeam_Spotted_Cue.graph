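# Sound cue for the deployable briefcase turret's targeting beam alarm, evidently played when the turret spots a target.
# A root mixer layers a pitched-down spin wave with a second mixer delayed by a fixed fraction of a second.
# The delayed mixer blends two alarm waves, one pitched down, and has an empty first input.
# It uses the turret's sight attenuation asset and overrides concurrency to one instance per owner, blocking new plays.
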SoundCue /DeployableTurretGameplay/Items/DeployableTurret/Sounds/BriefcaseTurret_TargetBeam_Spotted_Cue
root=Mixer_1
nodes [
 {id=Delay_0, children [Mixer_2], properties {delayMax=0.35, delayMin=0.35}}
 {id=Mixer_1, children [Delay_0,Modulator_3], properties {inputVolume [1.1,1.3]}}
 {id=Mixer_2, children [null,WavePlayer_8,Modulator_4], properties {inputVolume [1,1.2,0.8]}}
 {id=Modulator_3, children [WavePlayer_6], properties {pitchMax=0.8, pitchMin=0.8}}
 {id=Modulator_4, children [WavePlayer_4], properties {pitchMax=0.7, pitchMin=0.7}}
 {
  id=WavePlayer_4
  properties {soundWaveAssetPtr=&/BriefcaseTurret_TargetingBeam_Alarm_TargetSpoted_05}
 }
 {
  id=WavePlayer_6
  properties {soundWaveAssetPtr=&/BriefcaseTurret_TargetingBeam_Alarm_TargetSpoted_Spin_02}
 }
 {
  id=WavePlayer_8
  properties {soundWaveAssetPtr=&/BriefcaseTurret_TargetingBeam_Alarm_TargetSpoted_08}
 }
]
properties {
 attenuationSettings=/DeployableTurretGameplay/Items/DeployableTurret/Sounds/Settings/BriefcaseTurret_Sight_Att
 bOverrideConcurrency=true
 maxDistance=4100
 soundClassObject=&Fort_Gadgets_Items
 volumeMultiplier=0.8
 attenuationOverrides {
  bEnableFocusInterpolation=true
  bEnableListenerFocus=true
  falloffDistance=1200
  nonFocusVolumeAttenuation=0.5
  stereoSpread=500
  attenuationShapeExtents {x=0}
 }
 concurrencyOverrides {bLimitToOwner=true, maxCount=1, resolutionRule=PreventNew}
}
