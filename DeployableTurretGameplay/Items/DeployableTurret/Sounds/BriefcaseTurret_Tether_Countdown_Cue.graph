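# Sound cue for the deployable briefcase turret's tether countdown sound, routed through the critical in-game UI sound class.
# A single wave player plays one countdown wave directly, with no randomization or mixing, at a reduced volume.
# An overridden concurrency allows one instance per owner and stops the oldest when a new one starts.
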SoundCue /DeployableTurretGameplay/Items/DeployableTurret/Sounds/BriefcaseTurret_Tether_Countdown_Cue
root=WavePlayer_0
nodes [{
  id=WavePlayer_0
  properties {soundWaveAssetPtr=&/BriefcaseTurret_Tether_Countdown_B_01}
 }]
properties {
 bOverrideConcurrency=true
 maxDistance=2097152
 soundClassObject=&Fort_UI_InGame_Critical
 volumeMultiplier=0.1
 concurrencyOverrides {bLimitToOwner=true, maxCount=1, resolutionRule=StopOldest}
}
