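# Sound cue for the deployable briefcase turret, named for the moment its tether text appears, routed to the critical in-game UI sound class.
# A Random node picks between two equally weighted wave variations and feeds a Modulator at the root.
# A concurrency override allows one instance per owner, stopping the oldest when it retriggers.
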SoundCue /DeployableTurretGameplay/Items/DeployableTurret/Sounds/BriefcaseTurret_Tether_TextAppears_Cue
root=Modulator_0
nodes [
 {id=Modulator_0, children [Random_0]}
 {id=Random_0, children [WavePlayer_1,WavePlayer_2], properties {weights [1,1]}}
 {
  id=WavePlayer_1
  properties {soundWaveAssetPtr=&/BriefcaseTurret_Tether_TextAppears_01}
 }
 {
  id=WavePlayer_2
  properties {soundWaveAssetPtr=&/BriefcaseTurret_Tether_TextAppears_02}
 }
]
properties {
 bOverrideConcurrency=true
 maxDistance=2097152
 soundClassObject=&Fort_UI_InGame_Critical
 volumeMultiplier=1
 concurrencyOverrides {bLimitToOwner=true, maxCount=1, resolutionRule=StopOldest}
}
