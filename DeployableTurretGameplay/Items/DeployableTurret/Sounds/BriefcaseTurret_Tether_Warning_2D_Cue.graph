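# Sound cue for the deployable briefcase turret's 2D tether warning, routed to the critical in-game UI sound class.
# It picks one of three equally weighted "distant" warning waves at random and passes it through a Modulator node at default settings, with the cue at reduced volume.
# A concurrency override allows one instance per owner, and a new play stops the oldest.
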SoundCue /DeployableTurretGameplay/Items/DeployableTurret/Sounds/BriefcaseTurret_Tether_Warning_2D_Cue
root=Modulator_0
nodes [
 {id=Modulator_0, children [Random_1]}
 {id=Random_1, children [WavePlayer_0,WavePlayer_4,WavePlayer_5], properties {weights [1,1,1]}}
 {
  id=WavePlayer_0
  properties {soundWaveAssetPtr=&/BriefcaseTurret_Tether_Warning_Distant_03}
 }
 {
  id=WavePlayer_4
  properties {soundWaveAssetPtr=&/BriefcaseTurret_Tether_Warning_Distant_02}
 }
 {
  id=WavePlayer_5
  properties {soundWaveAssetPtr=&/BriefcaseTurret_Tether_Warning_Distant_01}
 }
]
properties {
 bOverrideConcurrency=true
 maxDistance=2097152
 soundClassObject=&Fort_UI_InGame_Critical
 volumeMultiplier=0.5
 concurrencyOverrides {bLimitToOwner=true, maxCount=1, resolutionRule=StopOldest}
}
