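# Sound cue for the throw sound of the deployable briefcase turret, classed under gadget item sounds.
# A Random node picks among six equally weighted throw recordings and feeds a Modulator at the root, with the cue at reduced volume.
# It borrows the turret's Sight attenuation asset for spatial falloff.
# Its own concurrency override allows one instance per owner and blocks new plays while one is active.
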SoundCue /DeployableTurretGameplay/Items/DeployableTurret/Sounds/BriefcaseTurret_Throw_Cue
root=Modulator_0
nodes [
 {id=Modulator_0, children [Random_0]}
 {id=Random_0, children [WavePlayer_0,WavePlayer_2,WavePlayer_3,WavePlayer_4,WavePlayer_5,WavePlayer_6], properties {weights [1,1,1,1,1,1]}}
 {id=WavePlayer_0, properties {soundWaveAssetPtr=&/BriefcaseTurret_Throw_01}}
 {id=WavePlayer_2, properties {soundWaveAssetPtr=&/BriefcaseTurret_Throw_02}}
 {id=WavePlayer_3, properties {soundWaveAssetPtr=&/BriefcaseTurret_Throw_03}}
 {id=WavePlayer_4, properties {soundWaveAssetPtr=&/BriefcaseTurret_Throw_04}}
 {id=WavePlayer_5, properties {soundWaveAssetPtr=&/BriefcaseTurret_Throw_05}}
 {id=WavePlayer_6, properties {soundWaveAssetPtr=&/BriefcaseTurret_Throw_06}}
]
properties {
 attenuationSettings=/DeployableTurretGameplay/Items/DeployableTurret/Sounds/Settings/BriefcaseTurret_Sight_Att
 bOverrideConcurrency=true
 maxDistance=4100
 soundClassObject=&Fort_Gadgets_Items
 volumeMultiplier=0.5
 attenuationOverrides {
  bEnableFocusInterpolation=true
  bEnableListenerFocus=true
  falloffDistance=1200
  nonFocusVolumeAttenuation=0.5
  stereoSpread=500
  attenuationShapeExtents {x=0}
 }
 concurrencyOverrides {bLimitToOwner=true, maxCount=1, resolutionRule=PreventNew}
}
