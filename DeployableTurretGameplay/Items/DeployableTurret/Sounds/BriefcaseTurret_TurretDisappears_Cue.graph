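# Sound cue for the deployable briefcase turret's disappearance, played as a spatialized gadget item sound with ambient attenuation.
# A Random node picks one of three equally weighted tether disappearance waves and feeds a Modulator for slight pitch and volume variation.
# A concurrency override allows one instance per owner, stopping the oldest when a new one plays.
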SoundCue /DeployableTurretGameplay/Items/DeployableTurret/Sounds/BriefcaseTurret_TurretDisappears_Cue
root=Modulator_0
nodes [
 {id=Modulator_0, children [Random_0]}
 {id=Random_0, children [WavePlayer_1,WavePlayer_2,WavePlayer_3], properties {weights [1,1,1]}}
 {
  id=WavePlayer_1
  properties {soundWaveAssetPtr=&/BriefcaseTurret_Tether_TurretDisappears_01}
 }
 {
  id=WavePlayer_2
  properties {soundWaveAssetPtr=&/BriefcaseTurret_Tether_TurretDisappears_02}
 }
 {
  id=WavePlayer_3
  properties {soundWaveAssetPtr=&/BriefcaseTurret_Tether_TurretDisappears_03}
 }
]
properties {
 attenuationSettings=/DeployableTurretGameplay/Items/DeployableTurret/Sounds/Settings/BriefcaseTurret_Ambient_Att
 bOverrideConcurrency=true
 maxDistance=1050
 soundClassObject=&Fort_Gadgets_Items
 volumeMultiplier=1.2
 concurrencyOverrides {bLimitToOwner=true, maxCount=1, resolutionRule=StopOldest}
}
